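# Niagara aura effect for the PistolCanary TripleTentacleFloorAttack, attached to sockets or bones of the User.SK skeletal mesh.
# Aura_small bursts per-socket glow sprites with array-driven sizes, fading over life and dimming at night.
# Spline_Mesh_Interpolate_ShardGlow places partitioned meshes along those sockets, length-scaled by IsLargeBody.
# Each mesh's smoothed position and a front-end-aware intensity feed an Interpolate Dynamic parameter, likely for trailing deformation.
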NiagaraSystem /PistolCanary/Assets/FX/TripleTentacleFloorAttack/NS_PistolCanary_TripleTentacleFloorAttack_Aura
dependencies {
 usedDataInterfaces [
  &NiagaraDataInterfaceArrayFloat
  &NiagaraDataInterfaceArrayFloat3
  &NiagaraDataInterfaceCurve
  &NiagaraDataInterfaceArrayInt32
  &NiagaraDataInterfaceArrayColor
 ]
 usedDynamicInputs []
 usedModules [
  &SystemState
  &self:SystemSpawnScript.SystemScriptSource.SystemScriptGraph.NiagaraNodeAssignment_1.SetVariables_#1
 ]
}
emitters {
 Aura_small {
  emitterUpdateScript [
   &EmitterState(Life Cycle Mode=::NewEnumerator1, Loop Behavior=::NewEnumerator1, Loop Duration=25, UseLoopDelay=false)
   &SpawnBurst_Instantaneous(Spawn Count=13, Use Spawn Probability=false, UseLoopCountLimit=false)
  ]
  particleSpawnScript [
   &FN_SetSpawnGroup(Number of Spawn Groups=15)
   &InitializeParticle(
    Color=<rgba 0.8721 0.438 2 1>
    Lifetime=25
    Sprite Size Mode=::NewEnumerator3
    Sprite UV Mode=::NewEnumerator1
    Uniform Sprite Size=&SelectFloatFromArray(
     Direct Array Index=$Particles.SpawnGroup
     Float Selection Array=DataInterface() {propertyValues {FloatData [70,70,70,70,70,50,50,100,100,80,80,70,70]}}
    )
    UsePositionOffset=false
    Write Material Random=true
   )
   &CameraOffset(Camera Offset Amount=-10)
  ]
  particleUpdateScript [
   &ParticleState()
   &SkeletalMeshLocation(
    Bone / Socket Index=$Particles.SpawnGroup
    Bone Sampling Mode=::NewEnumerator4
    MaskSpawnGroup=false
    Sampled Position Offset=&SelectVectorFromArray(
     Direct Array Index=$Particles.SpawnGroup
     Vector Selection Array=DataInterface() {propertyValues {FloatData [<0 0 0>, <0 0 0>, <0 0 0>, <0 0 0>, <0 0 0>, <5 0 0>, <-5 0 0>, <0 0 0>, <0 0 0>, <0 0 0>, <0 0 0>, <0 0 0>, <0 0 0>]}}
    )
    Skeletal Mesh=$User.SK
   )
   &ScaleSpriteSize(
    Scale Factor=&RandomRangeVector2D(Minimum=<0.55 0.65>, Maximum=<0.55 0.65>)
    Scale Sprite Size Mode=::NewEnumerator2
   ) {module=ScaleSpriteSize001}
   &ScaleSpriteSize(Scale Sprite Size Mode=::NewEnumerator0, Uniform Scale Factor=2)
   &ScaleColor(
    Scale Alpha=&FloatFromCurve(FloatCurve=DataInterface() {propertyValues {Curve {keys [
         {tangentMode=RCTM_User, time=0, value=0}
         <at 0.05 1>
         {arriveTangent=-0.0405, interpMode=RCIM_Cubic, leaveTangent=-0.0405, tangentMode=RCTM_User, time=0.4, value=1}
         {arriveTangent=-1.4489, interpMode=RCIM_Cubic, leaveTangent=-1.4489, tangentMode=RCTM_User, time=1, value=0}
        ]}}})
    Scale RGB=&VectorFromFloat(Value=&FNTimeOfDayScalar(Night Scale=0.75, Morning Scale=1, Evening Scale=1))
    ScaleA=true
    ScaleRGB=true
   )
  ]
  rendererClasses [&NiagaraSpriteRendererProperties]
  renderers [{rendererClass=&NiagaraSpriteRendererProperties}]
 }
 Spline_Mesh_Interpolate_ShardGlow {
  emitterUpdateScript [
   &EmitterState(
    Inactive Response=::NewEnumerator1
    Life Cycle Mode=::NewEnumerator1
    Loop Behavior=::NewEnumerator1
    Loop Duration Mode=::NewEnumerator1
    UseLoopDelay=false
   )
   &SpawnBurst_Instantaneous(Spawn Count=7, Use Spawn Probability=false, UseLoopCountLimit=false)
  ]
  particleSpawnScript [
   &PartitionParticles(Number Of Partitions=7)
   &InitializeParticle(
    Color=<rgba 1.125 1.408 3 1>
    Mesh Scale=&MakeVector(
     X=1
     Y=1
     Z=&Lerp_Float(
      A=&SelectFloatFromArray(
       Direct Array Index=$Particles.PartitionParticles.Partition
       Float Selection Array=DataInterface() {propertyValues {FloatData [1.35,1.2,1.2,0.8,0.8,0.5,0.5]}}
      )
      B=&SelectFloatFromArray(
       Direct Array Index=$Particles.PartitionParticles.Partition
       Float Selection Array=DataInterface() {propertyValues {FloatData [1.25,1.25,1.25,0.9,0.9,0.75,0.75]}}
      )
      Alpha=&MakeCustomFloatFromBool(Bool=$User.IsLargeBody, True Float=1, False Float=0)
     )
    )
    Mesh Scale Mode=::NewEnumerator4
    UsePositionOffset=false
    Write Material Random=true
   )
   &InitialMeshOrientation(Mesh Orientation Mode=::NewEnumerator1, Rotation=<0 0 0.3>, Use Rotation Vector=true)
   &CameraOffset()
  ]
  particleUpdateScript [
   &ParticleState(Kill Particles When Lifetime Has Elapsed=false, Let Infinitely Lived Particles Die When Emitter Deactivates=true)
   &DynamicMaterialParameters(
    Index 0 Param 1=&SelectFloatFromArray(
     Direct Array Index=$Particles.PartitionParticles.Partition
     Float Selection Array=DataInterface() {propertyValues {FloatData [-40,-40,-40,-30,-30,-30,-30]}}
    )
    Index 3 Param 1=1
    Index 3 Param 2=1
    Index 3 Param 3=1
    Index 3 Param 4=1
    Module.Param0WriteEnabled=true
    Module.Param0WriteEnabled3=true
    Module.Param1WriteEnabled=true
    Module.Param1WriteEnabled3=true
    Module.Param2WriteEnabled=true
    Module.Param2WriteEnabled3=true
    Module.Param3WriteEnabled=true
    Module.Param3WriteEnabled3=true
    Write Parameter Index 3=true
   )
   &UpdateMeshOrientation(
    Facing Direction=/Game/Effects/Niagara/DynamicInputs/Engine/FN_GetSystemAxis(Axis:ENiagaraOrientationAxis=::XAxis, Flip Axis:NiagaraBool=false)
    Facing Interpolation=false
    Orientation Method=::NewEnumerator1
    Override Quat=false
    TransformFacing=false
    TransformUp=false
   ) {module=UpdateMeshOrientation001}
   &SkeletalMeshLocation(
    Bone / Socket Index=&SelectIntFromArray(Direct Array Index=$Particles.PartitionParticles.Partition, Int Selection Array=DataInterface() {propertyValues {IntData [0,1,2,3,4,5,6,7]}})
    Bone Sampling Mode=::NewEnumerator4
    MaskSpawnGroup=false
    Orientation Sampling=::NewEnumerator1
    Sampled Position Offset=&SelectVectorFromArray(
     Direct Array Index=$Particles.PartitionParticles.Partition
     Vector Selection Array=DataInterface() {propertyValues {FloatData [<0 0 -10>, <0 0 -10>, <0 0 -10>, <0 0 -10>, <0 0 -10>, <0 0 -10>, <0 0 -10>, <0 0 0>]}}
    )
    Sampled Position Offset Coordinate Space:ENiagaraSkelMeshTransforms=::NewEnumerator1
    Skeletal Mesh=$User.SK
   )
   &InterpolateOverTime(
    Rate Of Change=12
    Target Vector Value:Vector3f=&ConvertPositionToVector(Input Position=$Particles.Position, Convert To Absolute Large World Vector=false)
    Variable Type=::NewEnumerator2
   )
   /SetVariables(Particles.Interpolate Dynamic=&MakeFloat4FromVectorAndFloat(
     "Vector (RGB)"=$StackContext.InterpolateOverTime.InterpolateOverTime_V3_.Moving Average
     "Float (Alpha)"=&Add_Float(
      A=&MakeCustomFloatFromBool(Bool=$User.IsFrontEnd, True Float=2, False Float=$User.Interpolate Intensity)
      B=0.5
     )
    )) {isSetParameters=true}
   /SetVariables(Particles.color2:LinearColor=&MultiplyLinearColor(
     Linear Color=&SelectLinearColorFromArray(
      Array Sampling Mode=::NewEnumerator0
      Direct Array Index=?
      Color Selection Array=DataInterface() {propertyValues {ColorData [<rgba 0 0.5283 1 1>, <rgba 0 0.0967 0.6302 1>]}}
      Random Seed=0
     )
     Scale Factor=<1.5 1.5 1.5 1>
    )) {enabled=false, isSetParameters=true}
  ]
  rendererClasses [&NiagaraMeshRendererProperties]
  renderers [{rendererClass=&NiagaraMeshRendererProperties, rendererIndex=0}]
 }
}
summary {
 systemName=NS_PistolCanary_TripleTentacleFloorAttack_Aura
 emitters [
  {emitterName=Aura_small, rendererClasses [&NiagaraSpriteRendererProperties]}
  {emitterName=Spline_Mesh_Interpolate_ShardGlow, rendererClasses [&NiagaraMeshRendererProperties]}
 ]
 userVariables [
  User.bIsFrontEnd {type=NiagaraBool}
  User.bIsFrontEndPreview {type=NiagaraBool}
  User.Color {type=LinearColor, description="", defaultValue {struct=LinearColor, value=<rgba 0 7.2362 32 1>}}
  User.FadeIn {type=NiagaraFloat, defaultValue=1, description=""}
  User.Interpolate Intensity {type=NiagaraFloat, defaultValue=3.5, description=""}
  User.IsFrontEnd {type=NiagaraBool, defaultValue=true}
  User.IsLargeBody {type=NiagaraBool}
  User.LN Lifetime Scale {type=NiagaraFloat, defaultValue=1, description=""}
  User.LN Spawn Burst {type=NiagaraFloat, defaultValue=5, description=""}
  User.LN Sprite Scale {type=NiagaraFloat, defaultValue=1, description=""}
  User.RankProgression {type=NiagaraInt32}
  User.RibbonIDOffset {type=NiagaraInt32}
  User.SK {
   type=NiagaraDataInterfaceSkeletalMesh
   defaultValue {
    struct=NiagaraExt_VariableValue_DataInterface
    value {dataInterface=&self:NiagaraDataInterfaceSkeletalMesh_1}
   }
  }
  User.SkeletalMesh {
   type=NiagaraDataInterfaceSkeletalMesh
   defaultValue {
    struct=NiagaraExt_VariableValue_DataInterface
    value {dataInterface=&self:NiagaraDataInterfaceSkeletalMesh_2}
   }
  }
 ]
}
