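# Niagara charge-up effect for PistolCanary's triple tentacle floor attack.
# User.Charge drives spawn rates, sizes and brightness across swirling energy and airflow meshes, inward-pulled sparks, corona, glows and a light, while some layers also grow with loop age.
# A GPU flare uses a scratch occlusion module and camera-angle scaling.
# Two mesh layers dim by time of day, User.Fade only scales the middle glow, and the skeletal-mesh arcs emitter is disabled.
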NiagaraSystem /PistolCanary/Assets/FX/TripleTentacleFloorAttack/NS_PistolCanary_TripleTentacleFloorAttack_Charge
dependencies {
 usedDataInterfaces [
  &NiagaraDataInterfaceCurve
  &NiagaraDataInterfaceColorCurve
  /Script/Niagara.NiagaraDataInterfaceOcclusion
  &NiagaraDataInterfaceCamera
  &NiagaraDataInterfaceArrayFloat
 ]
 usedDynamicInputs [&SimulationPosition, &ConvertVectorToPosition]
 usedModules [
  &self:SystemSpawnScript.SystemScriptSource.SystemScriptGraph.NiagaraNodeAssignment_0.SetVariables_#1
  &SystemState
 ]
}
emitters {
 Airflow001 {
  emitterUpdateScript [
   &EmitterState()
   &SpawnBurst_Instantaneous(Spawn Count=5, Use Spawn Probability=false, UseLoopCountLimit=false) {enabled=false}
   &SpawnRate(SpawnRate=&Multiply_Float(A=$User.Charge, B=10), Use Spawn Probability=true) {module=SpawnRate001}
  ]
  particleSpawnScript [
   &InitializeParticle(
    Color=<rgba 0.0216 0.5114 0.8021 4>
    Lifetime=&RandomRangeFloat(Minimum=0.9, Maximum=0.9)
    Mass Mode=::NewEnumerator1
    Mesh Scale=&Multiply_VectorByFloat(Vector=&RandomRangeVector(Minimum=<0.9 0.9 0.9>, Maximum=<1 1 1>), Float=3)
    Mesh Scale Mode=::NewEnumerator4
    Position Offset=&AddVectorToPosition(Position=$System.FXCenter, Vector=<0 0 0>)
    Position Offset Coordinate Space=::Local
    Sprite Size=<5 5>
    Sprite Size Mode=::NewEnumerator4
    UsePositionOffset=true
    Write Material Random=true
   ) {module=InitializeParticle001}
   &InitialMeshOrientation(Use Rotation Vector=false)
   &CameraOffset(Camera Offset Amount=100) {module=CameraOffset001}
  ]
  particleUpdateScript [
   &ParticleState()
   &ScaleMeshSize(Scale Factor=&VectorFromFloat(Value=&FloatFromCurve(
      FloatCurve=DataInterface() {propertyValues {Curve {keys [
          {arriveTangent=0.5023, interpMode=RCIM_Cubic, leaveTangent=0.5023, tangentMode=RCTM_User, time=0.2, value=0.5}
          {arriveTangent=0.3808, interpMode=RCIM_Cubic, leaveTangent=1.7513, tangentMode=RCTM_Break, time=0.7, value=0.7}
          {interpMode=RCIM_Cubic, tangentMode=RCTM_Break, time=0.8, value=0.9}
          {arriveTangent=-7.6859, interpMode=RCIM_Cubic, tangentMode=RCTM_Break, time=0.95, value=0}
         ]}}}
      CurveIndex=$Emitter.NormalizedLoopAge
      Scale Curve=&Multiply_Float(A=$User.Charge, B=1)
     ))) {module=ScaleMeshSize001}
   &SolveForcesAndVelocity(Clamp Velocity=false, Limit Acceleration=false)
   &DynamicMaterialParameters(
    Index 0 Param 1=&FloatFromCurve(FloatCurve=DataInterface() {propertyValues {Curve {keys [<at 0 -0.2>,<at 1 1.2>]}}})
    Index 0 Param 2=1
    Index 0 Param 3=&Multiply_Float(
     A=&RandomRangeFloat(Minimum=0.8, Maximum=1)
     B=&FloatFromCurve(FloatCurve=DataInterface() {propertyValues {Curve {keys [<at 0 1>,<at 1 0.5>]}}}, Scale Curve=20)
    )
    Module.Param0WriteEnabled=true
    Module.Param1WriteEnabled=true
    Module.Param2WriteEnabled=true
    Module.Param3WriteEnabled=true
   )
   &ScaleColor(
    Scale Alpha=&Multiply_Float(
     A=&Multiply_Float(
      A=&Multiply_Float(A=1, B=&FNTimeOfDayScalar(Night Scale=1, Morning Scale=0.3, Evening Scale=0.4))
      B=&FloatFromCurve(FloatCurve=DataInterface() {propertyValues {Curve {keys [
           <at 0 0 RCIM_Cubic>
           <at 0.1912 0.9932>
           {arriveTangent=-1.2279, interpMode=RCIM_Cubic, leaveTangent=-1.2279, time=0.7879, value=1.0115}
           <at 1 0>
          ]}}})
     )
     B=&FloatFromCurve(
      FloatCurve=DataInterface() {propertyValues {Curve {keys [
          {interpMode=RCIM_Cubic, tangentMode=RCTM_User, time=0, value=0}
          {arriveTangent=0.37, interpMode=RCIM_Cubic, leaveTangent=0.37, tangentMode=RCTM_User, time=0.6, value=0.05}
          {arriveTangent=1.1869, interpMode=RCIM_Cubic, leaveTangent=1.1869, tangentMode=RCTM_User, time=0.9, value=0.5258}
          <at 1.2 0.8>
         ]}}}
      CurveIndex=$Emitter.Age
     )
    )
    ScaleA=true
    ScaleRGB=true
   ) {enabled=false}
   &ScaleColor(
    Scale Alpha=&Multiply_Float(
     A=&Multiply_Float(
      A=&Multiply_Float(A=1, B=&FNTimeOfDayScalar(Night Scale=1, Morning Scale=0.3, Evening Scale=0.4))
      B=&FloatFromCurve(FloatCurve=DataInterface() {propertyValues {Curve {keys [<at 0.1912 0.9932>, {interpMode=RCIM_Cubic, leaveTangent=-1.2279, time=0.7879, value=1.0115}]}}})
     )
     B=$Emitter.NormalizedLoopAge
    )
    ScaleA=true
    ScaleRGB=true
   ) {module=ScaleColor002}
   &ScaleColor(
    Scale Alpha=&Multiply_Float(
     A=0.5
     B=&FloatFromCurve(FloatCurve=DataInterface() {propertyValues {Curve {keys [
          {arriveTangent=9.0725, interpMode=RCIM_Cubic, leaveTangent=9.0725, tangentMode=RCTM_User, time=0, value=0}
          {arriveTangent=-2.5249, interpMode=RCIM_Cubic, leaveTangent=-2.5249, tangentMode=RCTM_User, time=0.1408, value=0.9847}
          {arriveTangent=-0.0405, interpMode=RCIM_Cubic, leaveTangent=-0.0405, tangentMode=RCTM_User, time=1, value=0}
         ]}}})
    )
    Scale RGB=&Multiply_Vector(
     A=&MakeVectorFromLinearColorRGB(Color=&ColorFromCurve(ColorCurve=DataInterface() {propertyValues {
         AlphaCurve {keys [<at 0 1 RCIM_Cubic>,<at 0.9977 1 RCIM_Cubic>]}
         BlueCurve {keys [
           <at 0.0057 4.809>
           {arriveTangent=-9.9261, interpMode=RCIM_Cubic, leaveTangent=-9.9261, time=0.1454, value=1.0084}
           <at 0.4346 0.5512>
           <at 0.5279 2.0319>
           <at 0.7175 2.7561>
           {arriveTangent=0.9451, interpMode=RCIM_Cubic, leaveTangent=0.9451, time=0.8517, value=1.9922}
           <at 0.9448 2.9709>
          ]}
         GreenCurve {keys [
           <at 0.0057 4.7917>
           {arriveTangent=-9.9807, interpMode=RCIM_Cubic, leaveTangent=-9.9807, time=0.1454, value=0.9183}
           <at 0.4346 0.5104>
           <at 0.5279 1.762>
           <at 0.7175 2.5521>
           {arriveTangent=1.0311, interpMode=RCIM_Cubic, leaveTangent=1.0311, time=0.8517, value=1.7188}
           <at 0.9448 2.7865>
          ]}
         RedCurve {keys [
           <at 0.0057 5>
           {arriveTangent=-9.3251, interpMode=RCIM_Cubic, leaveTangent=-9.3251, time=0.1454, value=2}
           <at 0.4346 1>
           <at 0.5279 5>
           <at 0.7175 5>
           <at 0.8517 5 RCIM_Cubic>
           <at 0.9448 5>
          ]}
        }}))
     B=&VectorFromFloat(Value=2)
    )
    ScaleA=false
    ScaleRGB=true
   ) {module=ScaleColor001}
   &ScaleMeshSize(Scale Factor=&Multiply_Vector(
     A=<1 1 1>
     B=&VectorFromFloat(Value=&FloatFromCurve(
       FloatCurve=DataInterface() {propertyValues {Curve {keys [
           {arriveTangent=1.3114, interpMode=RCIM_Cubic, leaveTangent=1.3114, tangentMode=RCTM_User, time=0, value=0.2}
           {arriveTangent=1.0664, interpMode=RCIM_Cubic, leaveTangent=1.0664, tangentMode=RCTM_User, time=0.5, value=0.7}
           {arriveTangent=0.447, interpMode=RCIM_Cubic, leaveTangent=0.447, tangentMode=RCTM_User, time=1, value=1.2}
          ]}}}
       CurveIndex=$Emitter.Age
       Scale Curve=2
      ))
    ))
   &CameraOffset(Camera Offset Amount=20)
  ]
  rendererClasses [&NiagaraMeshRendererProperties]
  renderers [{rendererClass=&NiagaraMeshRendererProperties, rendererIndex=0}]
 }
 Arcs001 {
  bEnabled=false
  emitterUpdateScript [
   &EmitterState()
   &SpawnRate(SpawnRate=&Multiply_Float(A=$User.Charge, B=5), Use Spawn Probability=false)
  ]
  particleSpawnScript [
   &InitializeParticle(
    Lifetime Max=0.4
    Lifetime Min=0.6
    Lifetime Mode=::NewEnumerator1
    Sprite Rotation Mode=::NewEnumerator2
    Sprite Size Mode=::NewEnumerator3
    Uniform Sprite Size=&Multiply_Float(
     A=&FloatFromCurve(FloatCurve=DataInterface() {propertyValues {Curve {keys [<at 0 0.8>,<at 0.7 1.2>]}}}, CurveIndex=$Emitter.NormalizedLoopAge, Scale Curve=2)
     B=&RandomRangeFloat(Minimum=2000, Maximum=2500)
    )
    UsePositionOffset=false
    Write Material Random=true
   )
   &SubUVAnimation(End Frame=3, SubUV Animation Mode:ENiagara_SubUVLookupMode=::NewEnumerator2)
   &SkeletalMeshLocation(Bone / Socket Index=18, Bone Sampling Mode=::NewEnumerator6, MaskSpawnGroup=false, Skeletal Mesh=$System.Skeletal Mesh)
  ]
  particleUpdateScript [
   &ParticleState()
   &CameraOffset(Camera Offset Amount=25)
   &ScaleColor(
    Scale RGB=&MakeVectorFromLinearColorRGB(Color=&LerpLinearColors(
      StartColor=&ColorFromCurve(ColorCurve=DataInterface() {propertyValues {
         AlphaCurve {keys [<at 0.5586 1 RCIM_Cubic>,<at 1 0 RCIM_Cubic>]}
         BlueCurve {keys [<at 0 1 RCIM_Cubic>,<at 1 1 RCIM_Cubic>]}
         GreenCurve {keys [{arriveTangent=-0.0536, interpMode=RCIM_Cubic, time=0, value=0.2093}, {interpMode=RCIM_Cubic, leaveTangent=-0.0536, time=1, value=0.1557}]}
         RedCurve {keys [{arriveTangent=0.0676, interpMode=RCIM_Cubic, time=0, value=0.0396}, {interpMode=RCIM_Cubic, leaveTangent=0.0676, time=1, value=0.1073}]}
        }})
      EndColor=&ColorFromCurve(ColorCurve=DataInterface() {propertyValues {
         AlphaCurve {keys [<at 0.5586 1 RCIM_Cubic>,<at 1.0091 0 RCIM_Cubic>]}
         BlueCurve {keys [<at 0 1.2408 RCIM_Cubic>,<at 0.982 0>]}
         GreenCurve {keys [<at 0 1.7901 RCIM_Cubic>,<at 0.982 0.2738>]}
         RedCurve {keys [<at 0 5 RCIM_Cubic>,<at 0.982 5>]}
        }})
      LerpFactor=&SelectFloatFromArray(Direct Array Index=0, Float Selection Array=DataInterface() {propertyValues {FloatData [0,0,0,1]}})
     ))
    ScaleA=false
    ScaleRGB=true
   )
   &ScaleColor(
    Scale RGB=&VectorFromFloat(Value=&FloatFromCurve(FloatCurve=DataInterface() {propertyValues {Curve {keys [
          {interpMode=RCIM_Cubic, tangentMode=RCTM_User, time=0, value=1}
          {arriveTangent=106.638, interpMode=RCIM_Cubic, leaveTangent=-47.9454, tangentMode=RCTM_Break, time=0.2, value=10}
          {interpMode=RCIM_Cubic, tangentMode=RCTM_User, time=0.8, value=1}
         ]}}}))
    ScaleA=false
    ScaleRGB=true
   ) {module=ScaleColor001}
   &ScaleSpriteSize(Scale Factor=&Vector2DFromFloat(Value=&FloatFromCurve(FloatCurve=DataInterface() {propertyValues {Curve {keys [
          {arriveTangent=-0.0604, interpMode=RCIM_Cubic, leaveTangent=-0.0604, tangentMode=RCTM_User, time=0, value=0.6}
          {arriveTangent=1.6386, interpMode=RCIM_Cubic, leaveTangent=1.6386, tangentMode=RCTM_User, time=1, value=1}
         ]}}})))
   &SpriteRotationRate(Rotation Rate=25)
   &ScaleSpriteSize(
    Scale Factor=&MakeVector2D(
     X=&FloatFromCurve(FloatCurve=DataInterface() {propertyValues {Curve {keys [
          {arriveTangent=-3.4808, interpMode=RCIM_Cubic, leaveTangent=-3.4808, tangentMode=RCTM_User, time=0.4, value=1}
          {arriveTangent=0.0049, interpMode=RCIM_Cubic, leaveTangent=0.0049, tangentMode=RCTM_User, time=0.8, value=0.5}
         ]}}})
     Y=&FloatFromCurve(FloatCurve=DataInterface() {propertyValues {Curve {keys [
          {arriveTangent=0.0124, interpMode=RCIM_Cubic, leaveTangent=0.0124, tangentMode=RCTM_User, time=0.4, value=0.85}
          {arriveTangent=2.3888, interpMode=RCIM_Cubic, leaveTangent=2.3888, tangentMode=RCTM_User, time=0.8, value=1.2}
         ]}}})
    )
    Scale Sprite Size Mode=::NewEnumerator2
   ) {module=ScaleSpriteSize001}
   &DynamicMaterialParameters(
    Index 0 Param 1=1
    Index 0 Param 2=1
    Index 0 Param 3=&FloatFromCurve(FloatCurve=DataInterface() {propertyValues {Curve {keys [
         {arriveTangent=0.1411, interpMode=RCIM_Cubic, leaveTangent=0.1411, tangentMode=RCTM_User, time=0.2, value=0}
         {arriveTangent=1.8523, interpMode=RCIM_Cubic, leaveTangent=1.8523, tangentMode=RCTM_User, time=1, value=0.5}
        ]}}})
    Module.Param0WriteEnabled=true
    Module.Param1WriteEnabled=true
    Module.Param2WriteEnabled=true
    Module.Param3WriteEnabled=false
   )
  ]
  rendererClasses [&NiagaraSpriteRendererProperties]
  renderers [{rendererClass=&NiagaraSpriteRendererProperties}]
 }
 Corona {
  emitterUpdateScript [
   &EmitterState(Life Cycle Mode=::NewEnumerator1, Loop Behavior=::NewEnumerator1, Loop Duration=5.5, UseLoopDelay=false)
   &SpawnRate(
    SpawnRate=&Multiply_Float(
     A=5
     B=&FloatFromCurve(
      FloatCurve=DataInterface() {propertyValues {Curve {keys [{interpMode=RCIM_Cubic, leaveTangent=1.4489, tangentMode=RCTM_Break, time=0, value=0.3}, <at 1 1>]}}}
      CurveIndex=$System.NormalizedLoopAge
     )
    )
    Use Spawn Probability=false
   )
   &SpawnBurst_Instantaneous(Use Spawn Probability=false, UseLoopCountLimit=false) {enabled=false}
  ]
  particleSpawnScript [
   &InitializeParticle(
    Color=<rgba 5 0.2553 2.5482 1>
    Lifetime=0.3
    Position Offset=&AddVectorToPosition(Position=$System.FXCenter, Vector=<0 0 0>)
    Position Offset Coordinate Space=::Local
    Sprite Size Mode=::NewEnumerator3
    Uniform Sprite Size=&Multiply_Float(
     A=862.7131
     B=&FloatFromCurve(
      FloatCurve=DataInterface() {propertyValues {Curve {keys [
          {interpMode=RCIM_Cubic, leaveTangent=0.3896, tangentMode=RCTM_Break, time=0, value=0.3}
          {arriveTangent=1.3375, interpMode=RCIM_Cubic, leaveTangent=1.3375, tangentMode=RCTM_User, time=0.6, value=0.6925}
          <at 1 1>
         ]}}}
      CurveIndex=$System.NormalizedLoopAge
     )
    )
    UsePositionOffset=true
    Write Material Random=true
   )
   &CameraOffset(Camera Offset Amount=&Multiply_Float(A=50, B=1))
  ]
  particleUpdateScript [
   &ParticleState()
   &ScaleSpriteSize(
    Scale Sprite Size Mode=::NewEnumerator0
    Uniform Scale Factor=&FloatFromCurve(
     FloatCurve=DataInterface() {propertyValues {Curve {keys [
         {arriveTangent=0.4295, interpMode=RCIM_Cubic, leaveTangent=0.4295, tangentMode=RCTM_User, time=0, value=2.5}
         {arriveTangent=-1.7854, interpMode=RCIM_Cubic, leaveTangent=-1.7854, time=0.5336, value=1.3855}
         <at 0.9 0.8932 RCIM_Cubic>
        ]}}}
     Scale Curve=&Multiply_Float(A=1.5, B=$User.Charge)
    )
   )
   &ScaleColor(
    Scale Alpha=&Multiply_Float(
     A=1
     B=&FloatFromCurve(
      FloatCurve=DataInterface() {propertyValues {Curve {keys [
          {interpMode=RCIM_Cubic, tangentMode=RCTM_User, time=0, value=0}
          {interpMode=RCIM_Cubic, tangentMode=RCTM_User, time=0.1, value=1}
          {tangentMode=RCTM_User, time=1, value=0}
         ]}}}
      Scale Curve=&Multiply_Float(A=0.5, B=$System.NormalizedLoopAge)
     )
    )
    Scale RGB=&MakeVectorFromLinearColorRGB(Color=&ColorFromCurve(ColorCurve=DataInterface() {propertyValues {
        AlphaCurve {keys []}
        BlueCurve {keys [<at 0 0.904 RCIM_Cubic>,<at 0.9887 2.526 RCIM_Cubic>]}
        GreenCurve {keys [<at 0 0.983 RCIM_Cubic>,<at 0.9887 4.6572 RCIM_Cubic>]}
        RedCurve {keys [<at 0 1 RCIM_Cubic>,<at 0.9887 5 RCIM_Cubic>]}
       }}))
    ScaleA=true
    ScaleRGB=true
   )
   &DynamicMaterialParameters(
    Index 0 Param 1=&FloatFromCurve(FloatCurve=DataInterface() {propertyValues {Curve {keys [<at 0 1 RCIM_Cubic>,<at 1 1>]}}})
    Index 0 Param 2=1
    Index 0 Param 3=&FloatFromCurve(FloatCurve=DataInterface() {propertyValues {Curve {keys [<at 0 0.3>,<at 2 1>]}}})
    Index 0 Param 4=1
    Module.Index0Param0Write=true
    Module.Index0Param1Write=true
    Module.Index0Param2Write=true
    Module.Index0Param3Write=true
   )
  ]
  rendererClasses [&NiagaraSpriteRendererProperties]
  renderers [{rendererClass=&NiagaraSpriteRendererProperties}]
 }
 DrawnInSparks {
  emitterUpdateScript [
   &EmitterState()
   &SpawnRate(
    SpawnRate=&Lerp_Float(
     A=0
     B=100
     Alpha=&FloatFromCurve(
      FloatCurve=DataInterface() {propertyValues {Curve {keys [
          {interpMode=RCIM_Cubic, tangentMode=RCTM_User, time=0.2, value=0}
          {arriveTangent=2.5039, interpMode=RCIM_Cubic, leaveTangent=2.5039, tangentMode=RCTM_User, time=0.8, value=1}
         ]}}}
      CurveIndex=$User.Charge
     )
    )
    Use Spawn Probability=false
   )
  ]
  particleSpawnScript [
   &InitializeParticle(
    Color=<rgba 5 5 5 1>
    Lifetime Max=1.25
    Lifetime Min=0.8
    Lifetime Mode=::NewEnumerator1
    Sprite Size Max=<16 32>
    Sprite Size Min=<6 8>
    Sprite Size Mode=::NewEnumerator2
    UsePositionOffset=true
    Write Material Random=true
   )
   &ShapeLocation(
    Hemisphere Distribution=<0.525 1>
    MaskSpawnGroup=false
    Non Uniform Scale=<1 0.3 1>
    Override Seed=false
    Rotation Angle=-0.25
    Rotation Axis=<0 0 1>
    Rotation Mode=::NewEnumerator1
    Sphere Radius=750
    UseOwnerScale=true
   )
   &AddVelocityFromPoint(Module.Use Falloff Distance=false, Velocity Strength=&RandomRangeFloat(Minimum=300, Maximum=400))
   &CameraOffset(Camera Offset Amount=10)
  ]
  particleUpdateScript [
   &ParticleState()
   &CurlNoiseForce(
    Noise Frequency=200
    Noise Strength=&Multiply_Float(
     A=1500
     B=&FloatFromCurve(FloatCurve=DataInterface() {propertyValues {Curve {keys [<at 0 0>,<at 0.2 1>,<at 0.3 0>]}}})
    )
    Pan Noise=false
   )
   &Drag(Drag=&RandomRangeFloat(Minimum=1, Maximum=3), Use Linear Drag=true, Use Rotational Drag=false)
   &PointAttractionForce(Attraction Radius=800, Attraction Strength=375, Kill Within Radius=true, Module.Use Falloff=true, UseAttractorOffset=false)
   &SolveForcesAndVelocity(Clamp Velocity=false, Limit Acceleration=false)
   &ScaleSpriteSizeBySpeed(Max Scale Factor=<0.2 10>, Min Scale Factor=<1 0.5>)
   &ScaleSpriteSize(Scale Factor=&MakeVector2D(
     X=&FloatFromCurve(FloatCurve=DataInterface() {propertyValues {Curve {keys [<at 0 0>,<at 0.1 1>,<at 1 0>]}}})
     Y=&FloatFromCurve(FloatCurve=DataInterface() {propertyValues {Curve {keys [<at 0 1>,<at 1 0>]}}})
    ))
   &ScaleColor(
    Linear Color Curve=DataInterface() {propertyValues {
      AlphaCurve {keys [<at 0 1>]}
      BlueCurve {keys [<at 0 1>,<at 0.1583 1>,<at 0.3467 0.3723>,<at 0.5211 1>,<at 0.8598 1>]}
      GreenCurve {keys [<at 0 0.1833>,<at 0.1583 0>,<at 0.3467 0.0575>,<at 0.5211 0.0833>,<at 0.8598 0.6>]}
      RedCurve {keys [<at 0 0>,<at 0.1583 0.2833>,<at 0.3467 1>,<at 0.5211 0>,<at 0.8598 0>]}
     }}
    Scale Mode=::NewEnumerator2
   )
   &ScaleColor(
    Scale RGB=&VectorFromFloat(Value=&FloatFromCurve(FloatCurve=DataInterface() {propertyValues {Curve {keys [<at 0 0.5>,<at 0.4 1>,<at 1 0.5>]}}}, Scale Curve=20))
    ScaleA=true
    ScaleRGB=true
   ) {module=ScaleColor001}
  ]
  rendererClasses [&NiagaraSpriteRendererProperties]
  renderers [{rendererClass=&NiagaraSpriteRendererProperties}]
 }
 Energy_Large_Air {
  emitterUpdateScript [
   &EmitterState()
   &SpawnRate(
    SpawnRate=&Multiply_Float(
     A=&FloatFromCurve(
      FloatCurve=DataInterface() {propertyValues {Curve {keys [{interpMode=RCIM_Cubic, tangentMode=RCTM_Break, time=0, value=1}, {interpMode=RCIM_Cubic, tangentMode=RCTM_Break, time=5, value=8.8301}]}}}
      CurveIndex=$Emitter.Age
     )
     B=1
    )
    Use Spawn Probability=false
   )
  ]
  particleSpawnScript [
   &InitializeParticle(
    Color=<rgba 9.0601 1.0446 10 1>
    Lifetime=0.75
    Mesh Scale Max=<4 4 4>
    Mesh Scale Min=<2 2 2>
    Mesh Scale Mode=::NewEnumerator2
    Position Offset=&AddVectorToPosition(Position=$System.FXCenter, Vector=<0 0 0>)
    Position Offset Coordinate Space=::Local
    UsePositionOffset=true
    Write Material Random=true
   )
   &DynamicMaterialParameters(
    Index 0 Param 2=-1
    Index 0 Param 3=1
    Index 0 Param 4=1
    Module.Param0WriteEnabled=false
    Module.Param1WriteEnabled=true
    Module.Param2WriteEnabled=true
    Module.Param3WriteEnabled=true
   )
   &InitialMeshOrientation(Orientation Axis=<0 0 1>, Orientation Vector=<0 0 1>, Rotation=<0 0.25 0>, Use Orientation Vector=true, Use Rotation Vector=true) {module=InitialMeshOrientation001}
   &SystemLocation(MaskSpawnGroup=false, Offset=<2 -2 -2>) {enabled=false}
   &InitialMeshOrientation(Orientation Axis=<0 0 1>, Orientation Vector=<0 0 1>, Rotation=<-0.8 0 0>, Use Orientation Vector=true, Use Rotation Vector=true) {module=InitialMeshOrientation002}
   &InitialMeshOrientation(
    Rotation=&RandomRangeVector(Minimum=<0 0 0>, Maximum=<0 0 1>)
    Use Orientation Vector=true
    Use Rotation Vector=true
   )
  ]
  particleUpdateScript [
   &ParticleState()
   &ScaleMeshSize(Scale Factor=&VectorFromFloat(Value=&FloatFromCurve(
      FloatCurve=DataInterface() {propertyValues {Curve {keys [<at 0 0.5>,<at 0.2 1.2>,<at 0.8 1>,<at 1 0.5>]}}}
      Scale Curve=&Multiply_Float(A=$User.Charge, B=3)
     )))
   &DynamicMaterialParameters(
    Index 0 Param 1=0.55
    Module.Param0WriteEnabled=true
    Module.Param1WriteEnabled=false
    Module.Param2WriteEnabled=false
    Module.Param3WriteEnabled=false
   ) {module=DynamicMaterialParameters001}
   &ScaleColor(
    Scale Alpha=&FloatFromCurve(
     FloatCurve=DataInterface() {propertyValues {Curve {keys [<at 0 0>,<at 0.3 1>,<at 1 0>]}}}
     Scale Curve=&FloatFromCurve(FloatCurve=DataInterface() {propertyValues {Curve {keys [<at 0 0.5 RCIM_Cubic>,<at 0.2 1>]}}})
    )
    Scale RGB=&VectorFromFloat(Value=&FloatFromCurve(FloatCurve=DataInterface() {propertyValues {Curve {keys [<at 0 0>,<at 0.3 1>,<at 1 0>]}}}))
    ScaleA=true
    ScaleRGB=true
   )
   &ScaleColor(
    Scale Alpha=&FNTimeOfDayScalar(Night Scale=0.3, Morning Scale=0.3, Evening Scale=0.5)
    ScaleA=true
    ScaleRGB=true
   ) {module=ScaleColor001}
   &MeshRotationForce(Rotation=&MakeVector(
     X=0
     Y=0
     Z=&FloatFromCurve(FloatCurve=DataInterface() {propertyValues {Curve {keys [<at 0 1>,<at 1 0>]}}})
    ))
   &SolveForcesAndVelocity(Clamp Velocity=false, Limit Acceleration=false)
  ]
  rendererClasses [&NiagaraMeshRendererProperties]
  renderers [{rendererClass=&NiagaraMeshRendererProperties, rendererIndex=0}]
 }
 Energy_Offset_1 {
  emitterUpdateScript [
   &EmitterState()
   &SpawnRate(SpawnRate=6, Use Spawn Probability=false)
   /SetVariables(Emitter.Linear Color:LinearColor=&LerpLinearColors(
     StartColor=<rgba 0.0159 0.1401 1 1>
     EndColor=<rgba 0.7387 0.1312 1 1>
     LerpFactor=&Waveform("Waveform [1]"=::NewEnumerator0, "[1] Frequency"=2, "[1] Amplitude Min/Max"=<0 1>)
    )) {isSetParameters=true}
  ]
  particleSpawnScript [
   &InitializeParticle(Lifetime=2, Mesh Scale=<3 3 2>, Mesh Scale Mode=::NewEnumerator4, UsePositionOffset=true, Write Material Random=true)
   &InitialMeshOrientation(Mesh Axis To Orient=<0 0 1>, Mesh Orientation Mode=::NewEnumerator2, Use Rotation Vector=true)
   /SetVariables(Particles.MeshOrientation=&MultiplyQuaternion(
     Quaternion A=$Particles.InitialRotation
     Quaternion B=&MakeQuatFromAxisAngle(Axis=<0 0 1>, AngleInDegrees=&RandomRangeFloat(Minimum=-360, Maximum=360), NormalizeAxis=true)
    )) {isSetParameters=true}
  ]
  particleUpdateScript [
   &ParticleState()
   &ScaleMeshSize(Scale Factor=&VectorFromFloat(Value=&FloatFromCurve(FloatCurve=DataInterface() {propertyValues {Curve {keys [<at 0 0.5>,<at 0.2 1.2>,<at 0.8 1>,<at 1 0.5>]}}}, Scale Curve=$User.Charge)))
   &DynamicMaterialParameters(
    Index 0 Param 1=&FloatFromCurve(FloatCurve=DataInterface() {propertyValues {Curve {keys [{arriveTangent=2.4002, interpMode=RCIM_Cubic, leaveTangent=2.4002, tangentMode=RCTM_User, time=0, value=-0.63}, <at 1 0.7>]}}})
    Index 0 Param 2=&RandomRangeFloat(Minimum=-2, Maximum=1)
    Index 0 Param 3=0.1
    Module.Index0Param0Write=true
    Module.Index0Param1Write=true
    Module.Index0Param2Write=true
    Module.Index0Param3Write=false
   ) {module=DynamicMaterialParameters001}
   &ScaleColor(
    Scale Alpha=&FloatFromCurve(
     FloatCurve=DataInterface() {propertyValues {Curve {keys [<at 0 0>,<at 0.3 1>,<at 1 0>]}}}
     Scale Curve=&FloatFromCurve(FloatCurve=DataInterface() {propertyValues {Curve {keys [<at 0 0.5>,<at 0.2 1>]}}})
    )
    Scale RGB=&VectorFromFloat(Value=&FloatFromCurve(FloatCurve=DataInterface() {propertyValues {Curve {keys [<at 0 0>,<at 0.3 1>,<at 1 0>]}}}))
    ScaleA=true
    ScaleRGB=true
   )
   &UpdateMeshOrientation(Rotation Rate=0.025, Rotation Vector=<0 0 1>)
   &CameraOffset(Camera Offset Amount=350)
  ]
  rendererClasses [&NiagaraMeshRendererProperties]
  renderers [{rendererClass=&NiagaraMeshRendererProperties, rendererIndex=0}]
 }
 GPU_Flare {
  simTarget=GPUComputeSim
  emitterUpdateScript [
   &EmitterState()
   &SpawnBurst_Instantaneous(Use Spawn Probability=false, UseLoopCountLimit=true)
  ]
  particleSpawnScript [&InitializeParticle(
    Color=<rgba 0.7975 2.6478 6 1>
    Mesh Scale Mode=::NewEnumerator3
    Mesh Uniform Scale=20
    Sprite Rotation Mode=::NewEnumerator0
    Sprite Size Mode=::NewEnumerator3
    Uniform Sprite Size=1200
    UsePositionOffset=false
    Write Material Random=true
   )]
  particleUpdateScript [
   &ParticleState(Kill Particles When Lifetime Has Elapsed=false, Let Infinitely Lived Particles Die When Emitter Deactivates=true)
   &self:ScratchModule(
    NumberOfRings=5
    OcclusionSize=120
    Position:NiagaraPosition=&TransformPosition(Position=$Particles.Position, Source Space=::Local, Destination Space=::World)
    SamplesPerRing=5
    SystemX Sample Offset=5
   ) {module=ScratchModule}
   /SetVariables(Particles.CameraScale:NiagaraFloat=&MakeFloatFromVector(
     Vector=&Multiply_VectorByFloat(
      Vector=&ReturnCameraVector(
       Camera Vectors=::NewEnumerator4
       Coordinate Space=::Simulation
       Origin Position=$Particles.Position
       CameraQuery=DataInterface() {propertyValues {bRequireCurrentFrameData=false}}
      )
      Float=&DotProduct(
       Vector A=<1 0 1>
       Vector B=&ReturnCameraVector(
        Camera Vectors=::NewEnumerator4
        Coordinate Space=::Simulation
        Origin Position=$Particles.Position
        CameraQuery=DataInterface() {propertyValues {bRequireCurrentFrameData=false}}
       )
      )
     )
     Channel=::NewEnumerator0
    )) {isSetParameters=true}
   /SetVariables(Particles.CameraScale:NiagaraFloat=&Multiply_Float(
     A=$Particles.CameraScale
     B=&RemapRange(
      Input Value To Remap=&DotProduct(
       Vector A=&NormalizeVector(Vector To Normalize=<1 0 0>, Safe Normalize=true, Safe Fallback Threshold=0, Safe Fallback Vector=<0 0 0.001>)
       Vector B=&NormalizeVector(
        Vector To Normalize=&ReturnCameraVector(
         Camera Vectors=::NewEnumerator1
         Coordinate Space=::Simulation
         CameraQuery=DataInterface() {propertyValues {bRequireCurrentFrameData=false}}
        )
        Safe Normalize=true
        Safe Fallback Threshold=0
        Safe Fallback Vector=<0 0 0.001>
       )
      )
      Input Min=0
      Input Max=-1
      Output Min=0
      Output Max=1
     )
    )) {isSetParameters=true}
   &ScaleSpriteSize(
    Uniform Curve Index=$Particles.CameraScale
    Uniform Curve Scale=$User.Charge
    Uniform Curve Sprite Scale=DataInterface() {propertyValues {Curve {keys [
        {arriveTangent=4.7849, interpMode=RCIM_Cubic, leaveTangent=4.7849, tangentMode=RCTM_User, time=0, value=0.5}
        {arriveTangent=0.2873, interpMode=RCIM_Cubic, leaveTangent=0.2873, tangentMode=RCTM_User, time=0.1, value=1.4}
        {tangentMode=RCTM_User, time=1, value=1.5}
       ]}}}
   )
   &ScaleColor(Scale Alpha=$Particles.Visibility, Scale RGB=&VectorFromFloat(Value=3), ScaleA=true, ScaleRGB=true) {module=ScaleColor002}
   &ScaleColor(
    Scale Alpha=&FloatFromCurve(
     FloatCurve=DataInterface() {propertyValues {Curve {keys [
         {arriveTangent=5.9018, interpMode=RCIM_Cubic, leaveTangent=5.9018, tangentMode=RCTM_User, time=0, value=0}
         {arriveTangent=0.9356, leaveTangent=0.9356, time=0.05, value=1}
         {arriveTangentWeight=1.2316, interpMode=RCIM_Cubic, tangentMode=RCTM_User, time=1, value=1}
        ]}}}
     CurveIndex=$Particles.CameraScale
    )
    Scale RGB=&Multiply_VectorByFloat(Vector=<1 1 1>, Float=$Particles.CameraScale)
    ScaleA=true
    ScaleRGB=true
   )
  ]
  rendererClasses [&NiagaraSpriteRendererProperties]
  renderers [{rendererClass=&NiagaraSpriteRendererProperties}]
 }
 Glow {
  emitterUpdateScript [
   &EmitterState()
   &SpawnBurst_Instantaneous(Use Spawn Probability=false, UseLoopCountLimit=true)
  ]
  particleSpawnScript [&InitializeParticle(
    Color=<rgba 0.0129 0.0128 0.0365 0.5>
    Lifetime=0.9
    Sprite Size Mode=::NewEnumerator3
    Uniform Sprite Size=3000
    UsePositionOffset=true
    Write Material Random=true
   )]
  particleUpdateScript [
   &ParticleState(Kill Particles When Lifetime Has Elapsed=false, Let Infinitely Lived Particles Die When Emitter Deactivates=true)
   &ScaleSpriteSize(Uniform Curve Index=$User.Charge)
   /SetVariables(
    Particles.GlowColor2:LinearColor=&MakeLinearColorFromVectorAndFloat(
     "Vector (RGB)"=&MakeVectorFromLinearColorRGB(Color=<rgba 0.5427 0.2013 5 1>)
     "Float (Alpha)"=&Multiply_Float(A=0.2, B=1)
    )
    Particles.GlowSize2:Vector2f=&Vector2DFromFloat(Value=&Multiply_Float(A=1000, B=$User.Charge))
   ) {isSetParameters=true}
   &CameraOffset(Camera Offset Amount=300)
  ]
  rendererClasses [&NiagaraSpriteRendererProperties]
  renderers [
   {rendererClass=&NiagaraSpriteRendererProperties}
   {rendererClass=&NiagaraSpriteRendererProperties, rendererIndex=1}
  ]
 }
 Light_and_Flare {
  emitterUpdateScript [
   &EmitterState()
   &SpawnBurst_Instantaneous(Use Spawn Probability=false, UseLoopCountLimit=true)
  ]
  particleSpawnScript [
   &InitializeParticle(
    Lifetime Max=0.5
    Lifetime Min=0.25
    Lifetime Mode=::NewEnumerator1
    Sprite Size Mode=::NewEnumerator3
    Uniform Sprite Size=3200
    UsePositionOffset=false
    Write Material Random=true
   )
   &CameraOffset(Camera Offset Amount=290)
   &DynamicMaterialParameters(
    Index 0 Param 1=1
    Index 0 Param 2=1
    Index 0 Param 3=1
    Index 0 Param 4=1
    Index 1 Param 1=400
    Module.Index0Param0Write=true
    Module.Index0Param1Write=true
    Module.Index0Param2Write=true
    Module.Index0Param3Write=true
    Module.Index1Param0Write=true
    Module.Index1Param1Write=true
    Module.Index1Param2Write=true
    Module.Index1Param3Write=true
    Write Parameter Index 1=true
   )
  ]
  particleUpdateScript [
   &ParticleState(
    Kill Particles When Lifetime Has Elapsed=false
    Let Infinitely Lived Particles Die When Emitter Deactivates=true
    Loop Particles Lifetime=true
   )
   &SpriteRotationRate(Rotation Rate=2)
   &ScaleSpriteSize(Scale Sprite Size Mode=::NewEnumerator0, Uniform Scale Factor=&ClampFloat(Float=$User.Charge, Max=1.5))
   &ScaleColor(
    Scale Alpha=$User.Charge
    Scale RGB=&MakeVectorFromLinearColorRGB(Color=&LerpLinearColors(
      StartColor=<rgba 0 0.6167 1 1>
      EndColor=<rgba 0.7608 0 1 1>
      LerpFactor=&FloatFromCurve(FloatCurve=DataInterface() {propertyValues {Curve {keys [
           {interpMode=RCIM_Cubic, tangentMode=RCTM_User, time=0, value=0}
           {interpMode=RCIM_Cubic, tangentMode=RCTM_User, time=0.5, value=0.5}
           {tangentMode=RCTM_User, time=1, value=0}
          ]}}})
     ))
    ScaleA=true
    ScaleRGB=true
   )
   &FN_Light_Attributes(
    Color=&LerpLinearColors(
     StartColor=<rgba 0 0.8423 4 1>
     EndColor=<rgba 3.2923 0 4 1>
     LerpFactor=&RandomRangeFloat(Minimum=0.4, Maximum=0.6, Evaluation Type=::NewEnumerator1)
    )
    Create Color Attribute=true
    Create Exponent Attribute=true
    Create Radius Attribute=true
    Exponent=2
    Radius=350
    Scale Color=&Multiply_Float(A=&RandomRangeFloat(Minimum=3, Maximum=4, Evaluation Type=::NewEnumerator1), B=$User.Charge)
   )
  ]
  rendererClasses [&NiagaraSpriteRendererProperties, &NiagaraLightRendererProperties]
  renderers [
   {rendererClass=&NiagaraSpriteRendererProperties}
   {rendererClass=&NiagaraLightRendererProperties, rendererIndex=1}
  ]
 }
 MiddleGlow_Dark {
  emitterUpdateScript [
   &EmitterState()
   &SpawnBurst_Instantaneous(Spawn Time=$System.SpawnTime, Use Spawn Probability=false, UseLoopCountLimit=true)
  ]
  particleSpawnScript [
   &InitializeParticle(
    Color=<rgba 0.0091 0.0167 0.062 1>
    Lifetime=1.6
    Sprite Size=<4487.1797 8974.3594>
    Sprite Size Mode=::NewEnumerator4
    UsePositionOffset=true
    Write Material Random=true
   )
   &CameraOffset(Camera Offset Amount=2000)
  ]
  particleUpdateScript [
   &ParticleState(Kill Particles When Lifetime Has Elapsed=false, Let Infinitely Lived Particles Die When Emitter Deactivates=true)
   /SetVariables(Particles.Size002:Vector2f=&Multiply_Vector2D_ByFloat(
     Float=&ClampFloat(Float=$User.Charge, Max=2.5)
     Vector2D=&Multiply_Vector2D_ByFloat(
      Float=$User.Fade
      Vector2D=&Multiply_Vector2D(
       A=<2218.9351 2218.9351>
       B=&Multiply_Vector2D_ByFloat(
        Float=&Waveform("Waveform [1]"=::NewEnumerator12, "[1] Frequency"=9, "[1] Amplitude Min/Max"=<0.9 1.1>)
        Vector2D=<4 0.5>
       )
      )
     )
    )) {isSetParameters=true}
   /SetVariables(Particles.color002:LinearColor=&MakeLinearColorFromVectorAndFloat(
     "Vector (RGB)"=&Multiply_VectorByFloat(
      Vector=&MakeVectorFromLinearColorRGB(Color=&LerpLinearColors(
        StartColor=<rgba 1 0.0639 0.3455 1>
        EndColor=<rgba 1 0.0734 0.3551 1>
        LerpFactor=&Waveform("Waveform [1]"=::NewEnumerator12, "[1] Frequency"=3, "[1] Amplitude Min/Max"=<0 1>)
       ))
      Float=3
     )
     "Float (Alpha)"=&Multiply_Float(A=0.25, B=&Waveform("Waveform [1]"=::NewEnumerator12, "[1] Frequency"=3, "[1] Amplitude Min/Max"=<0.2 1>))
    )) {isSetParameters=true}
  ]
  rendererClasses [&NiagaraSpriteRendererProperties]
  renderers [{rendererClass=&NiagaraSpriteRendererProperties}]
 }
}
summary {
 systemName=NS_PistolCanary_TripleTentacleFloorAttack_Charge
 emitters [
  {emitterName=Energy_Offset_1, rendererClasses [&NiagaraMeshRendererProperties]}
  {emitterName=DrawnInSparks, rendererClasses [&NiagaraSpriteRendererProperties]}
  {emitterName=Glow, rendererClasses [&NiagaraSpriteRendererProperties]}
  {
   emitterName=Light_and_Flare
   rendererClasses [&NiagaraSpriteRendererProperties, &NiagaraLightRendererProperties]
  }
  {emitterName=GPU_Flare, simTarget=GPUComputeSim, rendererClasses [&NiagaraSpriteRendererProperties]}
  {emitterName=MiddleGlow_Dark, rendererClasses [&NiagaraSpriteRendererProperties]}
  {bEnabled=false, emitterName=Arcs001, rendererClasses [&NiagaraSpriteRendererProperties]}
  {emitterName=Airflow001, rendererClasses [&NiagaraMeshRendererProperties]}
  {emitterName=Corona, rendererClasses [&NiagaraSpriteRendererProperties]}
  {emitterName=Energy_Large_Air, rendererClasses [&NiagaraMeshRendererProperties]}
 ]
 userVariables [
  User.Alpha {type=NiagaraFloat, defaultValue=1, description=""}
  User.Charge {type=NiagaraFloat, defaultValue=3, description=""}
  User.Fade {type=NiagaraFloat, defaultValue=1, description=""}
 ]
}
