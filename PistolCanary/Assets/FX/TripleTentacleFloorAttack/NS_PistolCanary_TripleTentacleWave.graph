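# Niagara effect for PistolCanary's triple-tentacle floor attack: a cyan and blue energy shockwave burst with flash, sparks and lights.
# Burst-spawned emitters layer bolt meshes, a shock cone, energy wave sprites, expanding spheres, a flash mesh and drag-slowed curl-noise sparkles.
# Lifetime curves drive scale, color and dynamic material parameters, and camera-distance remaps adjust some sizes.
# The Dark_Glow and GroundWave emitters are disabled.
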NiagaraSystem /PistolCanary/Assets/FX/TripleTentacleFloorAttack/NS_PistolCanary_TripleTentacleWave
dependencies {
 usedDataInterfaces [
  &NiagaraDataInterfaceColorCurve
  &NiagaraDataInterfaceCurve
  &NiagaraDataInterfaceVectorCurve
  &NiagaraDataInterfaceArrayColor
  &NiagaraDataInterfacePlatformSet
  &NiagaraDataInterfaceArrayFloat
  &NiagaraDataInterfaceCamera
 ]
 usedDynamicInputs []
 usedModules [
  &self:SystemSpawnScript.SystemScriptSource.SystemScriptGraph.NiagaraNodeAssignment_0.SetVariables_#1
  &SystemState
 ]
}
emitters {
 Bolt {
  emitterUpdateScript [
   &EmitterState()
   &SpawnBurst_Instantaneous(Spawn Count=10, Use Spawn Probability=false, UseLoopCountLimit=false)
  ]
  particleSpawnScript [
   /SetVariables(Particles.MeshIndex=&RandomRangeInt(Minimum=0, Maximum=3)) {isSetParameters=true}
   &InitializeParticle(
    Color=&SelectLinearColorFromArray(
     Array Sampling Mode=::NewEnumerator1
     Direct Array Index=&RandomRangeInt(Minimum=0, Maximum=1)
     Color Selection Array=DataInterface() {propertyValues {ColorData [<rgba 0.6434 4.3286 5 1>]}}
     Random Seed=?
    )
    Lifetime=&RandomRangeFloat(Minimum=0.6, Maximum=0.5)
    Mesh Scale=<0 0 0>
    Mesh Scale Mode=::NewEnumerator4
    Sprite Size Mode=::NewEnumerator3
    Uniform Sprite Size=1200
    UsePositionOffset=false
    Write Material Random=true
   )
   &SystemLocation(MaskSpawnGroup=false, Offset=&RandomRangeVector(Minimum=<0 0 25>, Maximum=<0 0 25>))
   &InitialMeshOrientation(
    Mesh Orientation Mode=::NewEnumerator2
    Orientation Vector=<0 0 1>
    Rotation=&MakeVector(X=&RandomRangeFloat(Minimum=0, Maximum=1), Y=0, Z=0)
    Use Rotation Vector=true
   )
   &AddVelocityInCone(
    Cone Angle=90
    Cone Axis=<0 0 0.25>
    Module.Use Velocity Falloff On Cone Axis=false
    Velocity Distribution Along Cone Axis=0.25
    Velocity Strength=-25
   ) {enabled=false}
   &AddVelocity(Velocity=<0 0 1>) {enabled=false}
   /SetVariables(Particles.Scale=&Multiply_VectorByFloat(
     Vector=<2 2 2>
     Float=&RemapRange(
      Input Value To Remap=&FN_LOD_Distance_Remap(Start Distance=5000, End Distance=15000, Minimum Output=0.65, Maximum Output=1)
      Input Min=0
      Input Max=1
      Output Min=0
      Output Max=1
     )
    )) {isSetParameters=true}
   &CameraOffset(Camera Offset Amount=&Lerp_Float(
     A=-200
     B=0
     Alpha=&FN_LOD_Distance_Remap(Start Distance=0, End Distance=1500, Minimum Output=0, Maximum Output=1)
    )) {enabled=false}
  ]
  particleUpdateScript [
   &SolveForcesAndVelocity(Clamp Velocity=false, Limit Acceleration=false)
   &ParticleState()
   &ScaleColor(
    Scale Alpha=&FloatFromCurve(FloatCurve=DataInterface() {propertyValues {Curve {keys [
         {interpMode=RCIM_Cubic, tangentMode=RCTM_User, time=0, value=0}
         {interpMode=RCIM_Cubic, tangentMode=RCTM_User, time=0.5, value=1}
         {tangentMode=RCTM_User, time=0.6, value=0}
        ]}}})
    Scale RGB=&VectorFromFloat(Value=&FloatFromCurve(
      FloatCurve=DataInterface() {propertyValues {Curve {keys [
          {interpMode=RCIM_Cubic, tangentMode=RCTM_User, time=0, value=0}
          {arriveTangent=-3.0443, interpMode=RCIM_Cubic, leaveTangent=-3.0443, tangentMode=RCTM_User, time=0.1, value=0.95}
          {interpMode=RCIM_Cubic, tangentMode=RCTM_User, time=0.4, value=0.55}
          {tangentMode=RCTM_User, time=1, value=0}
         ]}}}
      Scale Curve=15
     ))
    ScaleA=true
    ScaleRGB=true
   )
   &DynamicMaterialParameters(
    Index 0 Param 1=70
    Index 0 Param 2=&FloatFromCurve(FloatCurve=DataInterface() {propertyValues {Curve {keys [
         {interpMode=RCIM_Cubic, tangentMode=RCTM_User, time=0, value=0}
         {arriveTangent=-2.9636, interpMode=RCIM_Cubic, leaveTangent=-2.9636, tangentMode=RCTM_User, time=0.4, value=0.95}
         {tangentMode=RCTM_User, time=0.9, value=0}
        ]}}})
    Index 0 Param 3=&FloatFromCurve(
     FloatCurve=DataInterface() {propertyValues {Curve {keys [
         {interpMode=RCIM_Cubic, tangentMode=RCTM_User, time=0, value=0}
         {tangentMode=RCTM_User, time=1.5, value=0.85}
         {interpMode=RCIM_Cubic, tangentMode=RCTM_User, time=2.5, value=1}
        ]}}}
     Scale Curve=2
    )
    Index 0 Param 4=1
    Index 1 Param 1=&FloatFromCurve(FloatCurve=DataInterface() {propertyValues {Curve {keys [{interpMode=RCIM_Cubic, tangentMode=RCTM_User, time=0, value=1}, {interpMode=RCIM_Cubic, tangentMode=RCTM_User, time=1, value=0}]}}})
    Index 1 Param 2=50
    Index 1 Param 3=1
    Index 1 Param 4=1
    Index 2 Param 1=&FloatFromCurve(FloatCurve=DataInterface() {propertyValues {Curve {keys [{interpMode=RCIM_Cubic, tangentMode=RCTM_User, time=0, value=1}, {tangentMode=RCTM_User, time=1, value=0.5}]}}})
    Module.Param0WriteEnabled=true
    Module.Param0WriteEnabled1=true
    Module.Param0WriteEnabled2=true
    Module.Param1WriteEnabled=true
    Module.Param1WriteEnabled1=true
    Module.Param1WriteEnabled2=true
    Module.Param2WriteEnabled=true
    Module.Param2WriteEnabled1=true
    Module.Param2WriteEnabled2=true
    Module.Param3WriteEnabled=true
    Module.Param3WriteEnabled1=true
    Module.Param3WriteEnabled2=true
    Write Parameter Index 1=true
    Write Parameter Index 2=true
   )
   &ScaleMeshSize(Scale Factor=&MakeVector(
     X=&FloatFromCurve(
      FloatCurve=DataInterface() {propertyValues {Curve {keys [
          {arriveTangent=-5.4159, interpMode=RCIM_Cubic, leaveTangent=-5.4159, tangentMode=RCTM_User, time=0, value=1.5}
          {interpMode=RCIM_Cubic, tangentMode=RCTM_User, time=1, value=1.5}
         ]}}}
      Scale Curve=3
     )
     Y=&FloatFromCurve(FloatCurve=DataInterface() {propertyValues {Curve {keys [
          {arriveTangent=73.8235, interpMode=RCIM_Cubic, leaveTangent=73.8235, tangentMode=RCTM_User, time=0, value=0}
          {tangentMode=RCTM_User, time=0.05, value=1}
         ]}}})
     Z=&FloatFromCurve(
      FloatCurve=DataInterface() {propertyValues {Curve {keys [
          {arriveTangent=9.6743, interpMode=RCIM_Cubic, leaveTangent=9.6743, tangentMode=RCTM_User, time=0, value=0}
          {interpMode=RCIM_Cubic, tangentMode=RCTM_User, time=0.1588, value=1.0402}
          {interpMode=RCIM_Cubic, tangentMode=RCTM_User, time=0.9122, value=1.3662}
          <at 1 1>
         ]}}}
      Scale Curve=2
     )
    ))
  ]
  rendererClasses [&NiagaraMeshRendererProperties]
  renderers [{rendererClass=&NiagaraMeshRendererProperties, rendererIndex=0}]
 }
 Cone_Shock_Light {
  emitterUpdateScript [
   &EmitterState()
   &SpawnBurst_Instantaneous(Use Spawn Probability=false, UseLoopCountLimit=false)
  ]
  particleSpawnScript [
   &InitializeParticle(
    Color=<rgba 0.5755 2.0733 9 1>
    Lifetime=0.8
    Mesh Scale=&Multiply_VectorByFloat(
     Vector=<5 1 1>
     Float=&FN_LOD_Distance_Remap(Start Distance=4500, End Distance=15000, Minimum Output=12, Maximum Output=12, Clamp Output=true)
    )
    Mesh Scale Mode=::NewEnumerator4
    Sprite Size Mode=::NewEnumerator3
    Uniform Sprite Size=&Multiply_Float(
     A=800
     B=&FN_LOD_Distance_Remap(Start Distance=4500, End Distance=22500, Minimum Output=20, Maximum Output=20, Clamp Output=true)
    )
    UsePositionOffset=true
    Write Material Random=true
   )
   &InitialMeshOrientation(Mesh Axis To Orient=<0 0 1>, Mesh Orientation Mode=::NewEnumerator2, Use Rotation Vector=true)
   &DynamicMaterialParameters(
    Index 0 Param 1=0.7
    Index 0 Param 2=0.3
    Index 0 Param 3=2
    Index 2 Param 1=300
    Index 3 Param 1=0.5
    Index 3 Param 2=3
    Index 3 Param 3=1
    Index 3 Param 4=1
    Module.Param0WriteEnabled=true
    Module.Param0WriteEnabled2=true
    Module.Param0WriteEnabled3=true
    Module.Param1WriteEnabled=true
    Module.Param1WriteEnabled2=false
    Module.Param1WriteEnabled3=true
    Module.Param2WriteEnabled=true
    Module.Param2WriteEnabled2=false
    Module.Param2WriteEnabled3=true
    Module.Param3WriteEnabled=false
    Module.Param3WriteEnabled2=false
    Module.Param3WriteEnabled3=true
    Write Parameter Index 2=true
    Write Parameter Index 3=true
   )
   &CameraOffset(Camera Offset Amount=2000)
  ]
  particleUpdateScript [
   &ParticleState()
   &ScaleSpriteSize(Uniform Curve Sprite Scale=DataInterface() {propertyValues {Curve {keys [{arriveTangent=6.2587, interpMode=RCIM_Cubic, leaveTangent=6.2587, tangentMode=RCTM_User, time=0, value=0}, <at 0.5284 1 RCIM_Cubic>]}}})
   &ScaleMeshSize(Scale Factor=&VectorFromCurve(
     VectorCurve=DataInterface() {propertyValues {
       XCurve {keys [
         {arriveTangent=1.5385, interpMode=RCIM_Cubic, time=0, value=0}
         {arriveTangent=0.1488, interpMode=RCIM_Cubic, leaveTangent=0.1488, tangentMode=RCTM_User, time=1, value=1}
        ]}
       YCurve {keys [
         {arriveTangent=2.3109, interpMode=RCIM_Cubic, leaveTangent=2.3109, tangentMode=RCTM_User, time=0, value=0}
         {arriveTangent=0.6124, interpMode=RCIM_Cubic, leaveTangent=0.6124, tangentMode=RCTM_User, time=1, value=1}
        ]}
       ZCurve {keys [
         {arriveTangent=2.3109, interpMode=RCIM_Cubic, leaveTangent=2.3109, tangentMode=RCTM_User, time=0, value=0}
         {arriveTangent=0.6124, interpMode=RCIM_Cubic, leaveTangent=0.6124, tangentMode=RCTM_User, time=1, value=1}
        ]}
      }}
     Scale Curve=<0.3 7 7>
    ))
   &ScaleColor(
    Scale Alpha=&FloatFromCurve(FloatCurve=DataInterface() {propertyValues {Curve {keys [
         {arriveTangent=48.0273, interpMode=RCIM_Cubic, leaveTangent=48.0273, tangentMode=RCTM_User, time=0, value=0}
         <at 0.05 1 RCIM_Cubic>
         {interpMode=RCIM_Cubic, leaveTangent=-3.1778, time=1, value=0}
        ]}}})
    Scale RGB=&MakeVectorFromLinearColorRGB(Color=&ColorFromCurve(ColorCurve=DataInterface() {propertyValues {
        AlphaCurve {keys []}
        BlueCurve {keys [<at 0.1026 1 RCIM_Cubic>,<at 0.959 0.1 RCIM_Cubic>]}
        GreenCurve {keys [<at 0.1026 1 RCIM_Cubic>,<at 0.959 0.0492 RCIM_Cubic>]}
        RedCurve {keys [<at 0.1026 1 RCIM_Cubic>,<at 0.959 0.053 RCIM_Cubic>]}
       }}))
    ScaleA=true
    ScaleRGB=true
   )
   /SetVariables(Particles.FlareColor=&MultiplyLinearColor(
     Linear Color=$Particles.Color
     Scale Factor=&MakeVector4(
      X=1
      Y=1
      Z=1
      W=&FloatFromCurve(FloatCurve=DataInterface() {propertyValues {Curve {keys [<at 0 1>,<at 1 0>]}}})
     )
    )) {isSetParameters=true}
   /SetVariables(Particles.MeshColor:LinearColor=&MultiplyLinearColor(
     Linear Color=$Particles.Color
     Scale Factor=&MakeVector4(
      X=1
      Y=1
      Z=1
      W=&FloatFromCurve(FloatCurve=DataInterface() {propertyValues {Curve {keys [<at 0 0.3>,<at 1 0>]}}})
     )
    )) {isSetParameters=true}
   &DynamicMaterialParameters(
    Index 1 Param 1=&FloatFromCurve(FloatCurve=DataInterface() {propertyValues {Curve {keys [<at 0 -1 RCIM_Cubic>,<at 1 0>]}}})
    Module.Param0WriteEnabled1=true
    Module.Param1WriteEnabled1=true
    Module.Param2WriteEnabled1=false
    Module.Param3WriteEnabled1=false
    Write Parameter Index 0=false
    Write Parameter Index 1=true
   ) {module=DynamicMaterialParameters001}
   &FN_Light_Attributes(
    Color=<rgba 0.5899 2.008 4 1>
    Create Color Attribute=true
    Create Exponent Attribute=true
    Create Position Attribute=true
    Create Radius Attribute=true
    Exponent=2
    Position=&AddVectorToPosition(
     Position=$Particles.SkeletalMeshLocation.SampledPosition
     Vector=&RotateVectorByQuaternion(Vector To Rotate=<0 50 0>, Quaternion=$Particles.SkeletalMeshLocation.SampledRotation)
    )
    Radius=2000
    Scale Color=&FloatFromCurve(
     FloatCurve=DataInterface() {propertyValues {Curve {keys [
         {arriveTangent=1.5581, interpMode=RCIM_Cubic, time=0, value=0}
         {arriveTangent=0.1339, interpMode=RCIM_Cubic, leaveTangent=0.1339, tangentMode=RCTM_User, time=0.05, value=1}
         {arriveTangent=-2.9133, interpMode=RCIM_Cubic, leaveTangent=-2.9133, tangentMode=RCTM_User, time=1, value=0}
        ]}}}
     Scale Curve=6
    )
   )
  ]
  rendererClasses [
   &NiagaraMeshRendererProperties
   &NiagaraSpriteRendererProperties
   &NiagaraLightRendererProperties
  ]
  renderers [
   {rendererClass=&NiagaraMeshRendererProperties, rendererIndex=0}
   {rendererClass=&NiagaraSpriteRendererProperties, rendererIndex=1}
   {rendererClass=&NiagaraLightRendererProperties, rendererIndex=2}
  ]
 }
 Dark_Glow {
  bEnabled=false
  emitterUpdateScript [
   &EmitterState(
    Inactive Response=::NewEnumerator1
    Life Cycle Mode=::NewEnumerator1
    Loop Behavior=::NewEnumerator1
    Loop Delay=0.1
    Loop Duration=4
    UseLoopDelay=false
   )
   &SpawnBurst_Instantaneous(Spawn Count=1, Use Spawn Probability=false)
  ]
  particleSpawnScript [
   &InitializeParticle(
    Color=<rgba 1 1 1 1.5>
    Lifetime=2
    Sprite Rotation Angle=0.12
    Sprite Rotation Mode=::NewEnumerator1
    Sprite Size Mode=::NewEnumerator3
    Uniform Sprite Size=1500
    UsePositionOffset=true
    Write Material Random=true
   )
   &DynamicMaterialParameters(
    Index 0 Param 1=2
    Index 0 Param 2=1
    Index 0 Param 3=1
    Index 0 Param 4=1
    Module.Param0WriteEnabled=true
    Module.Param1WriteEnabled=true
    Module.Param2WriteEnabled=true
    Module.Param3WriteEnabled=true
   )
   &CameraOffset(Camera Offset Amount=10)
   &SpriteFacingAndAlignment(Sprite Facing=<0 0 1>, WriteAlignment=false, WriteFacing=true)
   &InitialMeshOrientation(Mesh Orientation Mode=::NewEnumerator2, Orientation Vector=<0 0 1>, Use Rotation Vector=true)
   /SetVariables(Particles.CameraOffset02:NiagaraFloat=&FN_PickFloatDependOnPlatform(TargetPlatformActive=10, TargetPlatformNoActive=-50)) {isSetParameters=true}
  ]
  particleUpdateScript [
   &ParticleState()
   &ScaleSpriteSize(
    Scale Sprite Size Mode=::NewEnumerator0
    Uniform Scale Factor=&Multiply_Float(
     A=&Multiply_Float(A=0.9, B=$System.SizeScale)
     B=&FloatFromCurve(
      FloatCurve=DataInterface() {propertyValues {Curve {keys [<at 0 0>,<at 1 1>]}}}
      Scale Curve=&Multiply_Float(
       A=$System.SizeScale
       B=&FN_LOD_Distance_Remap(Start Distance=350, End Distance=380, Minimum Output=0.7, Maximum Output=1, Clamp Output=true)
      )
     )
    )
   ) {module=ScaleSpriteSize001}
   &SolveForcesAndVelocity(Clamp Velocity=false, Limit Acceleration=false)
   &ScaleColor(
    Scale Alpha=&Multiply_Float(
     A=&FN_LOD_Distance_Remap(Start Distance=350, End Distance=380, Minimum Output=1, Maximum Output=1, Clamp Output=true)
     B=&FloatFromCurve(
      FloatCurve=DataInterface() {propertyValues {Curve {keys [<at 0 1>,<at 0.2 1 RCIM_Cubic>,<at 0.7 1 RCIM_Cubic>,<at 1 1 RCIM_Cubic>]}}}
      Scale Curve=1.1
     )
    )
    Scale RGB=&MakeVectorFromLinearColorRGB(Color=&ColorFromCurve(ColorCurve=DataInterface() {propertyValues {
        AlphaCurve {keys []}
        BlueCurve {keys [<at 0 0.0917 RCIM_Cubic>]}
        GreenCurve {keys [<at 0 0.1 RCIM_Cubic>]}
        RedCurve {keys [<at 0 0.0219 RCIM_Cubic>]}
       }}))
    ScaleA=true
    ScaleRGB=true
   )
   &FN_Light_Attributes(
    Color=<rgba 0 20 16.6667 1>
    Create Color Attribute=true
    Create Exponent Attribute=true
    Create Position Attribute=true
    Create Radius Attribute=true
    Enable Particle Light Enable=true
    Exponent=3
    Position=&TransformPosition(
     Position=&AddVectorToPosition(Position=$Particles.Position, Vector=<0 80 -50>)
     Source Space=::Local
     Destination Space=::Simulation
    )
    Radius=100
    Scale Color=&FloatFromCurve(FloatCurve=DataInterface() {propertyValues {Curve {keys [<at 0 1>,<at 0.7 1>,<at 1 0>]}}})
   )
   &FN_Sprite_Attributes(
    Color=&Multiply_LinearColorByFloat(
     Linear Color=<rgba 0 1.136 1.2 1>
     Float=&FloatFromCurve(FloatCurve=DataInterface() {propertyValues {Curve {keys [<at 0 1>,<at 1 0>]}}})
    )
    Create Color Attribute=true
    Create SpriteSize Attribute=true
    SpriteSize=&Multiply_Vector2D(
     A=&Vector2DFromFloat(Value=800)
     B=&Vector2DFromFloat(Value=&Multiply_Float(
       A=&FloatFromCurve(FloatCurve=DataInterface() {propertyValues {Curve {keys [<at 0 0>,<at 1 1>]}}}, Scale Curve=$System.SizeScale)
       B=&Multiply_Float(
        A=1.2
        B=&FN_LOD_Distance_Remap(Start Distance=350, End Distance=380, Minimum Output=0.7, Maximum Output=1, Clamp Output=true)
       )
      ))
    )
    Write to Module Namespace=true
   )
   &FN_Sprite_Attributes(
    Color=&MakeLinearColorFromVectorAndFloat(
     "Vector (RGB)"=&Multiply_Vector(
      A=&MakeVectorFromLinearColorRGB(Color=<rgba 1 1 1 1>)
      B=&Multiply_VectorByFloat(Vector=<1 1 1>, Float=0.8)
     )
     "Float (Alpha)"=&Multiply_Float(A=5, B=&FloatFromCurve(FloatCurve=DataInterface() {propertyValues {Curve {keys [<at 0 1>,<at 1 0>]}}}))
    )
    Create Color Attribute=true
    Create SpriteSize Attribute=true
    SpriteSize=&Multiply_Vector2D(
     A=&Vector2DFromFloat(Value=200)
     B=&Vector2DFromFloat(Value=&Multiply_Float(
       A=1
       B=&FloatFromCurve(
        FloatCurve=DataInterface() {propertyValues {Curve {keys [<at 0 0>,<at 0.9 1>]}}}
        Scale Curve=&Multiply_Float(
         A=$System.SizeScale
         B=&FN_LOD_Distance_Remap(Start Distance=350, End Distance=380, Minimum Output=0.7, Maximum Output=1, Clamp Output=true)
        )
       )
      ))
    )
    Write to Module Namespace=true
   ) {module=FN_Sprite_Attributes001}
   &FN_Mesh_Attributes(
    Color=&MakeLinearColorFromVectorAndFloat(
     "Vector (RGB)"=&Multiply_Vector(A=&MakeVectorFromLinearColorRGB(Color=<rgba 3 3 3 1>), B=<1 1 1>)
     "Float (Alpha)"=&Multiply_Float(A=1.5, B=1)
    )
    Create Color Attribute=true
    Create Scale Attribute=true
    Scale=&Multiply_Vector(
     A=<1 0.01 1>
     B=&VectorFromFloat(Value=&FloatFromCurve(
       FloatCurve=DataInterface() {propertyValues {Curve {keys [<at 0 0>,<at 0.05 1>,<at 0.83 1>,<at 0.9 0>]}}}
       Scale Curve=&Multiply_Float(A=$System.SizeScale, B=6.3)
      ))
    )
    Write to Module Namespace=true
   )
  ]
  rendererClasses [
   &NiagaraLightRendererProperties
   &NiagaraSpriteRendererProperties
   &NiagaraMeshRendererProperties
  ]
  renderers [
   {rendererClass=&NiagaraLightRendererProperties, rendererIndex=0}
   {rendererClass=&NiagaraSpriteRendererProperties, rendererIndex=1}
   {rendererClass=&NiagaraSpriteRendererProperties, rendererIndex=2}
   {rendererClass=&NiagaraMeshRendererProperties, rendererIndex=3}
  ]
 }
 EnergyWave {
  emitterUpdateScript [
   &EmitterState(Life Cycle Mode=::NewEnumerator1, Loop Behavior=::NewEnumerator1, Loop Delay=0.15, Loop Duration=1, UseLoopDelay=false)
   &SpawnBurst_Instantaneous(Use Spawn Probability=false, UseLoopCountLimit=false)
  ]
  particleSpawnScript [&InitializeParticle(
    Color=<rgba 4 4 4 2>
    Mesh Scale Mode=::NewEnumerator3
    Mesh Uniform Scale=-3.5
    Sprite Rotation Angle=90
    Sprite Rotation Mode=::NewEnumerator0
    Sprite Size Mode=::NewEnumerator3
    Uniform Sprite Size=&Multiply_Float(A=2161.01, B=10)
    UsePositionOffset=true
    Write Material Random=true
   )]
  particleUpdateScript [
   &ParticleState()
   &CameraOffset(Camera Offset Amount=&Multiply_Float(
     A=500
     B=&FloatFromCurve(
      FloatCurve=DataInterface() {propertyValues {Curve {keys [<at 0 0>,<at 1 1>]}}}
      CurveIndex=/Game/Effects/Niagara/Modules/Math/FN_CameraDistance(
       0Dist:NiagaraFloat=500
       1Dist:NiagaraFloat=2000
       Camera Query:NiagaraDataInterfaceCamera=DataInterface()
       Position:NiagaraPosition=$Particles.Position
      )
     )
    ))
   &ScaleSpriteSize(
    Uniform Curve Scale=2
    Uniform Curve Sprite Scale=DataInterface() {propertyValues {Curve {keys [
        {arriveTangent=3.1312, interpMode=RCIM_Cubic, leaveTangent=3.1312, tangentMode=RCTM_User, time=0, value=0}
        {tangentMode=RCTM_User, time=1, value=1}
       ]}}}
   )
   &DynamicMaterialParameters(
    Index 0 Param 1=1
    Index 0 Param 2=&FloatFromCurve(FloatCurve=DataInterface() {propertyValues {Curve {keys [
         {interpMode=RCIM_Cubic, tangentMode=RCTM_User, time=0, value=1}
         {arriveTangent=-1, interpMode=RCIM_Cubic, leaveTangent=-1, time=0.6, value=0.7}
         {interpMode=RCIM_Cubic, leaveTangent=-3.3753, time=1, value=0}
        ]}}})
    Index 0 Param 3=1
    Index 0 Param 4=1
    Module.Param0WriteEnabled=true
    Module.Param1WriteEnabled=true
    Module.Param2WriteEnabled=true
    Module.Param3WriteEnabled=true
   )
   &ScaleColor(
    Scale Alpha=&FloatFromCurve(
     FloatCurve=DataInterface() {propertyValues {Curve {keys [
         <at 0 0 RCIM_Cubic>
         {arriveTangent=-2.878, interpMode=RCIM_Cubic, leaveTangent=-2.878, tangentMode=RCTM_User, time=0.2, value=1}
         {arriveTangent=-1.25, interpMode=RCIM_Cubic, leaveTangent=-1.25, time=0.35, value=0.2}
         {interpMode=RCIM_Cubic, tangentMode=RCTM_User, time=1, value=0}
        ]}}}
     Scale Curve=&Multiply_Float(A=50, B=1)
    )
    Scale RGB=&MakeVectorFromLinearColorRGB(Color=&MultiplyLinearColor(
      Linear Color=&ColorFromCurve(ColorCurve=DataInterface() {propertyValues {
         AlphaCurve {keys []}
         BlueCurve {keys [<at 0.0364 1 RCIM_Cubic>, <at 0.1445 2>, {interpMode=RCIM_Cubic, leaveTangent=-0.7175, time=0.3716, value=0.1}]}
         GreenCurve {keys [<at 0.0364 0.6981 RCIM_Cubic>, <at 0.1445 0.3641>, {interpMode=RCIM_Cubic, leaveTangent=-4.4353, time=0.3716, value=0.0199}]}
         RedCurve {keys [<at 0.0364 0.0764 RCIM_Cubic>, <at 0.1445 0.144>, {interpMode=RCIM_Cubic, leaveTangent=2.2176, time=0.3716, value=0.0946}]}
        }})
      Scale Factor=&MakeFloat4FromVectorAndFloat("Vector (RGB)"=&VectorFromFloat(Value=10), "Float (Alpha)"=1)
     ))
    ScaleA=true
    ScaleRGB=true
   )
   /SetVariables(Particles.Color02:LinearColor=&MakeLinearColorFromVectorAndFloat(
     "Vector (RGB)"=&MakeVectorFromLinearColorRGB(Color=&MultiplyLinearColor(
       Linear Color=<rgba 1.1531 1.1714 8 1>
       Scale Factor=&MakeFloat4FromVectorAndFloat("Vector (RGB)"=&VectorFromFloat(Value=3), "Float (Alpha)"=1)
      ))
     "Float (Alpha)"=&FloatFromCurve(
      FloatCurve=DataInterface() {propertyValues {Curve {keys [
          <at 0 0 RCIM_Cubic>
          {arriveTangent=-1.6399, interpMode=RCIM_Cubic, leaveTangent=-1.6399, tangentMode=RCTM_User, time=0.15, value=1}
          {interpMode=RCIM_Cubic, tangentMode=RCTM_User, time=1, value=0}
         ]}}}
      Scale Curve=2
     )
    )) {isSetParameters=true}
   &FN_Light_Attributes(
    Color=<rgba 1 0.156 0.1861 1>
    Create Color Attribute=true
    Create Position Attribute=true
    Create Radius Attribute=true
    Position=/Niagara/DynamicInputs/Helpers/SimulationPosition()
    Radius=1491.5984
    Scale Color=&FloatFromCurve(
     FloatCurve=DataInterface() {propertyValues {Curve {keys [<at 0 0>, {interpMode=RCIM_Cubic, leaveTangent=-2.9073, tangentMode=RCTM_Break, time=0.1, value=1}, <at 0.8 0>]}}}
     Scale Curve=4449615.5
    )
   )
   &SpriteFacingAndAlignment(Sprite Facing=<0 0 1>, WriteAlignment=true, WriteFacing=true)
  ]
  rendererClasses [&NiagaraSpriteRendererProperties, &NiagaraLightRendererProperties]
  renderers [
   {rendererClass=&NiagaraSpriteRendererProperties}
   {rendererClass=&NiagaraSpriteRendererProperties, rendererIndex=1}
   {rendererClass=&NiagaraLightRendererProperties, rendererIndex=2}
  ]
 }
 Energy_Sphere {
  emitterUpdateScript [
   &EmitterState()
   &SpawnBurst_Instantaneous(Use Spawn Probability=false, UseLoopCountLimit=true)
   &SpawnBurst_Instantaneous(Spawn Group=1, Use Spawn Probability=false, UseLoopCountLimit=true) {module=SpawnBurst_Instantaneous001}
   &SpawnBurst_Instantaneous(Spawn Group=2, Use Spawn Probability=false, UseLoopCountLimit=true) {module=SpawnBurst_Instantaneous002}
  ]
  particleSpawnScript [
   &InitializeParticle(
    Lifetime=0.75
    Mesh Scale=&Multiply_VectorByFloat(
     Vector=&MaskVectorBySpawnGroup(Masked Vector=<13.2 13.2 13.2>, Passthrough Vector=<-20 -17.5 -17.5>)
     Float=&FN_LOD_Distance_Remap(Start Distance=1000, End Distance=22500, Minimum Output=10, Maximum Output=10, Clamp Output=true)
    )
    Mesh Scale Mode=::NewEnumerator4
    UsePositionOffset=true
    Write Material Random=true
   )
   /SetVariables(Particles.VisibilityTag=&SpawnGroup()) {isSetParameters=true, module=SetVariables_#2}
   &InitialMeshOrientation(
    Mesh Orientation Mode=::NewEnumerator2
    Rotation=&MaskVectorBySpawnGroup(
     Masked Vector=&RandomRangeVector(Minimum=<0.4 0 0>, Maximum=<0.6 0 0>)
     Passthrough Vector=<0 0 0>
     Spawn Group Mask=2
    )
    Use Rotation Vector=true
   )
  ]
  particleUpdateScript [
   &ParticleState()
   &ScaleColor(
    Scale Alpha=&MaskFloatBySpawnGroup(
     Masked Float=&FloatFromCurve(FloatCurve=DataInterface() {propertyValues {Curve {keys [{arriveTangent=-2.4108, interpMode=RCIM_Cubic, leaveTangent=-2.4108, tangentMode=RCTM_User, time=0.3, value=1}, <at 1 0 RCIM_Cubic>]}}})
     Passthrough Float=&MaskFloatBySpawnGroup(
      Masked Float=&FloatFromCurve(
       FloatCurve=DataInterface() {propertyValues {Curve {keys [
           {arriveTangent=-34.5605, interpMode=RCIM_Cubic, leaveTangent=-34.5605, tangentMode=RCTM_User, time=0, value=3}
           {arriveTangent=-2.3039, interpMode=RCIM_Cubic, leaveTangent=-2.3039, tangentMode=RCTM_User, time=0.0917, value=0.7188}
           {arriveTangent=-0.7914, interpMode=RCIM_Cubic, leaveTangent=-0.7914, time=0.5, value=0.1}
           {arriveTangent=0.0005, interpMode=RCIM_Cubic, leaveTangent=0.0005, tangentMode=RCTM_User, time=1, value=0}
          ]}}}
       Scale Curve=0.5
      )
      Passthrough Float=&FloatFromCurve(
       FloatCurve=DataInterface() {propertyValues {Curve {keys [
           {arriveTangent=-0.0222, interpMode=RCIM_Cubic, leaveTangent=-0.0222, tangentMode=RCTM_User, time=0.2, value=0}
           {arriveTangent=7.4942, interpMode=RCIM_Cubic, leaveTangent=-1.7347, tangentMode=RCTM_Break, time=0.4, value=1}
           {arriveTangent=-1.6667, interpMode=RCIM_Cubic, leaveTangent=-1.6667, time=0.6, value=0.375}
           {arriveTangent=-0.2805, interpMode=RCIM_Cubic, leaveTangent=-0.2805, tangentMode=RCTM_User, time=1, value=0}
          ]}}}
       Scale Curve=0.1
      )
      Spawn Group Mask=1
     )
    )
    Scale RGB=&MakeVectorFromLinearColorRGB(Color=&ColorFromCurve(ColorCurve=DataInterface() {propertyValues {
        AlphaCurve {keys []}
        BlueCurve {keys [<at 0.078 5>,<at 0.9434 1>]}
        GreenCurve {keys [<at 0.078 1.3625>,<at 0.9434 1>]}
        RedCurve {keys [<at 0.078 0.15>,<at 0.9434 1>]}
       }}))
    ScaleA=true
    ScaleRGB=true
   )
   /SetVariables(Particles.Color002=&MakeLinearColorFromVectorAndFloat(
     "Vector (RGB)"=&MakeVectorFromLinearColorRGB(Color=&MultiplyLinearColor(
       Linear Color=<rgba 1 1 1 1>
       Scale Factor=&MakeVector4_FromLinearColor(LinearColor=&ColorFromCurve(ColorCurve=DataInterface() {propertyValues {
           AlphaCurve {keys []}
           BlueCurve {keys [<at 0.078 5>,<at 0.9391 1>]}
           GreenCurve {keys [<at 0.078 1.3625>,<at 0.9391 1>]}
           RedCurve {keys [<at 0.078 0.15>,<at 0.9391 1>]}
          }}))
      ))
     "Float (Alpha)"=&FloatFromCurve(FloatCurve=DataInterface() {propertyValues {Curve {keys [<at 0.5 1>,<at 1 0>]}}}, Scale Curve=1.5)
    )) {isSetParameters=true}
   /SetVariables(
    Particles.Color3:LinearColor=&ColorFromCurve(ColorCurve=DataInterface() {propertyValues {
       AlphaCurve {keys [<at 0 1 RCIM_Cubic>,<at 1 0 RCIM_Cubic>]}
       BlueCurve {keys [<at 0.0348 1 RCIM_Cubic>]}
       GreenCurve {keys [<at 0.0348 0.3304 RCIM_Cubic>]}
       RedCurve {keys [<at 0.0348 0.0441 RCIM_Cubic>]}
      }})
    Particles.Color4:LinearColor=&ColorFromCurve(
     ColorCurve=DataInterface() {propertyValues {
       AlphaCurve {keys [<at 0 0.6 RCIM_Cubic>,<at 1 0.4 RCIM_Cubic>]}
       BlueCurve {keys [<at 0 1 RCIM_Cubic>,<at 1 0.3028 RCIM_Cubic>]}
       GreenCurve {keys [<at 0 0.26 RCIM_Cubic>,<at 1 0.0808 RCIM_Cubic>]}
       RedCurve {keys [<at 0 0.03 RCIM_Cubic>,<at 1 1 RCIM_Cubic>]}
      }}
     CurveIndex=&Waveform("Waveform [1]"=::NewEnumerator0, "[1] Input Position"=-0.7, "[1] Frequency"=2, "[1] Amplitude Min/Max"=<0 1>)
    )
   ) {isSetParameters=true}
   /SetVariables(Particles.Color4:LinearColor=&MultiplyLinearColor(
     Linear Color=$Particles.Color4
     Scale Factor=&MakeVector4(
      X=1
      Y=1
      Z=1
      W=&FloatFromCurve(FloatCurve=DataInterface() {propertyValues {Curve {keys [<at 0 1>,<at 1 0>]}}})
     )
    )) {isSetParameters=true}
   &ScaleMeshSize(Scale Factor=&VectorFromFloat(Value=&FloatFromCurve(FloatCurve=DataInterface() {propertyValues {Curve {keys [
          {arriveTangent=3.4726, interpMode=RCIM_Cubic, leaveTangent=3.4726, tangentMode=RCTM_User, time=0, value=0}
          {arriveTangent=0.1704, interpMode=RCIM_Cubic, leaveTangent=0.1704, tangentMode=RCTM_User, time=1, value=1.25}
         ]}}})))
   &CameraOffset(Camera Offset Amount=&FloatFromCurve(FloatCurve=DataInterface() {propertyValues {Curve {keys [<at 0 -0.5>,<at 1 1>]}}}, Scale Curve=50))
   &DynamicMaterialParameters(
    Index 0 Param 3=&FloatFromCurve(FloatCurve=DataInterface() {propertyValues {Curve {keys [<at 0.5 0.4>,<at 1 1>]}}})
    Index 0 Param 4=&FloatFromCurve(
     FloatCurve=DataInterface() {propertyValues {Curve {keys [
         {arriveTangent=-1.1835, interpMode=RCIM_Cubic, leaveTangent=-1.1835, tangentMode=RCTM_User, time=0, value=1}
         {arriveTangent=-0.5267, interpMode=RCIM_Cubic, leaveTangent=-0.5267, tangentMode=RCTM_User, time=1, value=0}
        ]}}}
     Scale Curve=40
    )
    Module.Index0Param0Write=true
    Module.Index0Param1Write=true
    Module.Index0Param2Write=true
    Module.Index0Param3Write=true
   )
  ]
  rendererClasses [&NiagaraMeshRendererProperties]
  renderers [
   {rendererClass=&NiagaraMeshRendererProperties, rendererIndex=0}
   {rendererClass=&NiagaraMeshRendererProperties, rendererIndex=1}
   {rendererClass=&NiagaraMeshRendererProperties, rendererIndex=2}
   {rendererClass=&NiagaraMeshRendererProperties, rendererIndex=3}
   {rendererClass=&NiagaraMeshRendererProperties, rendererIndex=4}
  ]
 }
 Explode_Sparkles {
  emitterUpdateScript [
   &EmitterState(Life Cycle Mode=::NewEnumerator1, Loop Behavior=::NewEnumerator1, Loop Delay=0.1, Loop Duration=1, UseLoopDelay=false)
   &SpawnBurst_Instantaneous(Spawn Count=140, Use Spawn Probability=false, UseLoopCountLimit=false)
  ]
  particleSpawnScript [
   &InitializeParticle(
    Color=<rgba 0.469 3.3359 5 1>
    Lifetime=&RandomRangeFloat(Minimum=0.8, Maximum=2)
    Sprite Rotation Mode=::NewEnumerator0
    Sprite Size Max=<23.0769 230.7692>
    Sprite Size Min=<23.0769 153.8462>
    Sprite Size Mode=::NewEnumerator2
    UsePositionOffset=false
    Write Material Random=true
   )
   &SystemLocation(MaskSpawnGroup=false)
   &ShapeLocation(
    MaskSpawnGroup=false
    Override Seed=false
    Rotation Axis=<0 0 1>
    Rotation Mode=::NewEnumerator1
    Shape Primitive=::NewEnumerator4
    UseOwnerScale=true
   )
   &AddVelocity(
    Module.Use Falloff Distance=false
    Override Seed=false
    Velocity Mode=::NewEnumerator1
    Velocity Speed=&RandomRangeFloat(Minimum=12000, Maximum=20000)
   )
  ]
  particleUpdateScript [
   /SetVariables(Particles.Glow03_Color=&MakeLinearColorFromVectorAndFloat(
     "Vector (RGB)"=&MakeVectorFromLinearColorRGB(Color=&MultiplyLinearColor(Linear Color=<rgba 0.2142 0 1 0.2>, Scale Factor=&MakeVector4_FromFloat(Float=1)))
     "Float (Alpha)"=&Multiply_Float(A=1, B=1)
    )) {isSetParameters=true}
   /SetVariables(Particles.Glow03_Size=&Multiply_Vector2D_ByFloat(
     Float=&FloatFromCurve(
      FloatCurve=DataInterface() {propertyValues {Curve {keys [
          <at 0 0 RCIM_Cubic>
          <at 0.2 1.3 RCIM_Cubic>
          {arriveTangent=-0.1423, interpMode=RCIM_Cubic, leaveTangent=-0.1423, tangentMode=RCTM_User, time=1, value=0}
         ]}}}
      Scale Curve=&Multiply_Float(
       A=10
       B=&ClampFloat(
        Float=&FN_LOD_Distance_Remap(Start Distance=1000, End Distance=10000, Minimum Output=1, Maximum Output=30)
        Max=2
       )
      )
     )
     Vector2D=<3 4>
    )) {isSetParameters=true}
   &ParticleState()
   &CurlNoiseForce(
    Noise Strength=&Multiply_Float(
     A=900
     B=&FloatFromCurve(
      FloatCurve=DataInterface() {propertyValues {Curve {defaultValue=1, keys [{tangentMode=RCTM_User, time=0, value=0}, {tangentMode=RCTM_User, time=1, value=1}]}}}
      Scale Curve=20
     )
    )
    Pan Noise=true
    Pan Noise Field=<0 0 1>
   )
   &ScaleSpriteSize(Scale Factor=&Vector2DFromFloat(Value=&FloatFromCurve(
      FloatCurve=DataInterface() {propertyValues {Curve {keys [
          <at 0 0 RCIM_Cubic>
          <at 0.2 1.2 RCIM_Cubic>
          {arriveTangent=-2.7132, interpMode=RCIM_Cubic, leaveTangent=-2.7132, tangentMode=RCTM_User, time=1, value=0}
         ]}}}
      Scale Curve=&Multiply_Float(
       A=0.7
       B=&ClampFloat(
        Float=&FN_LOD_Distance_Remap(Start Distance=1000, End Distance=10000, Minimum Output=1, Maximum Output=20)
        Max=2
       )
      )
     )))
   &ScaleColor(
    Scale Alpha=&Multiply_Float(A=1, B=&FloatFromCurve(FloatCurve=DataInterface() {propertyValues {Curve {keys [<at 0 1>,<at 1 0>]}}}))
    Scale RGB=&Multiply_VectorByFloat(Vector=<2 2 2>, Float=1)
    ScaleA=true
    ScaleRGB=true
   )
   &Drag(
    Drag=&FloatFromCurve(
     FloatCurve=DataInterface() {propertyValues {Curve {defaultValue=1, keys [{tangentMode=RCTM_User, time=0, value=0}, {tangentMode=RCTM_User, time=0.9, value=1}]}}}
     Scale Curve=7
    )
    Use Linear Drag=true
    Use Rotational Drag=false
   )
   &SolveForcesAndVelocity(Clamp Velocity=false, Limit Acceleration=false)
   &ScaleSpriteSizeBySpeed(Max Scale Factor=<0.6 5>, Min Scale Factor=<0.6 1>, Velocity Threshold=60000)
  ]
  rendererClasses [&NiagaraSpriteRendererProperties]
  renderers [
   {rendererClass=&NiagaraSpriteRendererProperties}
   {rendererClass=&NiagaraSpriteRendererProperties, rendererIndex=1}
  ]
 }
 Flash {
  emitterUpdateScript [
   &EmitterState()
   &SpawnBurst_Instantaneous(Use Spawn Probability=false, UseLoopCountLimit=false)
  ]
  particleSpawnScript [
   &InitializeParticle(
    Color=<rgba 7 1.7017 5.5336 1>
    Lifetime=0.7
    Mesh Scale Mode=::NewEnumerator3
    Mesh Uniform Scale=70
    Position Offset=<0 0 200>
    Sprite Size Mode=::NewEnumerator3
    Uniform Sprite Size=400
    UsePositionOffset=true
    Write Material Random=true
   )
   &CameraOffset(Camera Offset Amount=40)
   &InitialMeshOrientation(Mesh Orientation Mode=::NewEnumerator2, Orientation Vector=<0 0 1>, Use Rotation Vector=false)
  ]
  particleUpdateScript [
   &ParticleState()
   &ScaleMeshSize(Scale Factor=&VectorFromFloat(Value=&FloatFromCurve(FloatCurve=DataInterface() {propertyValues {Curve {keys [
          {arriveTangent=0.7693, interpMode=RCIM_Cubic, leaveTangent=0.7693, tangentMode=RCTM_User, time=0, value=0.6}
          {arriveTangent=0.1864, interpMode=RCIM_Cubic, leaveTangent=0.1864, tangentMode=RCTM_User, time=1, value=0.8}
         ]}}})))
   &ScaleColor(
    Scale Alpha=&FloatFromCurve(
     FloatCurve=DataInterface() {propertyValues {Curve {keys [{arriveTangent=-2.4409, arriveTangentWeight=0.162, interpMode=RCIM_Cubic, leaveTangentWeight=0.2268, time=0.2, value=1}, <at 1 0>]}}}
     Scale Curve=2.5
    )
    ScaleA=true
    ScaleRGB=false
   )
   &DynamicMaterialParameters(
    Index 0 Param 1=&FloatFromCurve(FloatCurve=DataInterface() {propertyValues {Curve {keys [
         {interpMode=RCIM_Cubic, tangentMode=RCTM_User, time=0.5, value=0}
         {arriveTangent=3.8908, interpMode=RCIM_Cubic, leaveTangent=3.8908, tangentMode=RCTM_User, time=1, value=1}
        ]}}})
    Index 0 Param 2=&FloatFromCurve(FloatCurve=DataInterface() {propertyValues {Curve {keys [
         {arriveTangent=-0.6134, interpMode=RCIM_Cubic, leaveTangent=-0.6134, tangentMode=RCTM_User, time=0, value=3}
         {arriveTangent=-3.248, interpMode=RCIM_Cubic, leaveTangent=-3.248, tangentMode=RCTM_User, time=1, value=0.5}
        ]}}})
    Index 0 Param 3=&FloatFromCurve(FloatCurve=DataInterface() {propertyValues {Curve {keys [
         {arriveTangent=-1.2425, interpMode=RCIM_Cubic, time=0, value=1}
         {arriveTangent=-2.1091, interpMode=RCIM_Cubic, leaveTangent=-2.1091, tangentMode=RCTM_User, time=1, value=0}
        ]}}})
    Index 0 Param 4=&FloatFromCurve(
     FloatCurve=DataInterface() {propertyValues {Curve {keys [
         {arriveTangent=2.3563, interpMode=RCIM_Cubic, leaveTangent=2.3563, tangentMode=RCTM_User, time=0, value=0}
         {tangentMode=RCTM_User, time=1, value=1}
        ]}}}
     Scale Curve=2
    )
    Module.Index0Param0Write=true
    Module.Index0Param1Write=true
    Module.Index0Param2Write=true
    Module.Index0Param3Write=true
   ) {module=DynamicMaterialParameters001}
   &ScaleSpriteSize(Uniform Curve Sprite Scale=DataInterface() {propertyValues {Curve {keys [
        {arriveTangent=13.251, interpMode=RCIM_Cubic, leaveTangent=13.251, tangentMode=RCTM_User, time=0, value=0.5}
        {arriveTangent=-1.9983, interpMode=RCIM_Cubic, leaveTangent=-1.9983, tangentMode=RCTM_User, time=0.11, value=1}
        <at 0.9986 0>
       ]}}})
  ]
  rendererClasses [&NiagaraMeshRendererProperties]
  renderers [{rendererClass=&NiagaraMeshRendererProperties, rendererIndex=0}]
 }
 GroundWave {
  bEnabled=false
  emitterUpdateScript [
   &EmitterState()
   &SpawnBurst_Instantaneous(Use Spawn Probability=false, UseLoopCountLimit=false)
   &SpawnBurst_Instantaneous(Spawn Group=1, Spawn Time=0.1, Use Spawn Probability=false, UseLoopCountLimit=false) {module=SpawnBurst_Instantaneous001}
  ]
  particleSpawnScript [
   &InitializeParticle(
    Color=<rgba 1.25 1.25 1.25 1>
    Lifetime=&MaskFloatBySpawnGroup(Masked Float=0.5, Passthrough Float=1)
    Mesh Scale Mode=::NewEnumerator3
    Mesh Uniform Scale=&MaskFloatBySpawnGroup(Masked Float=2.5, Passthrough Float=3.5)
    UsePositionOffset=false
    Write Material Random=true
   )
   &InitialMeshOrientation(
    Mesh Orientation Mode=::NewEnumerator2
    Rotation=&RandomRangeVector(Minimum=<0 0 0>, Maximum=<0 0 1>)
    Use Rotation Vector=true
   )
   &CameraOffset(Camera Offset Amount=40)
  ]
  particleUpdateScript [
   &ParticleState()
   &ScaleMeshSize(Scale Factor=&VectorFromCurve(VectorCurve=DataInterface() {propertyValues {
       XCurve {keys [
         {arriveTangent=1.4912, interpMode=RCIM_Cubic, leaveTangent=1.4912, tangentMode=RCTM_User, time=0, value=0.2}
         {arriveTangent=0.1443, interpMode=RCIM_Cubic, leaveTangent=0.1443, tangentMode=RCTM_User, time=1, value=1}
        ]}
       YCurve {keys [
         {arriveTangent=1.4912, interpMode=RCIM_Cubic, leaveTangent=1.4912, tangentMode=RCTM_User, time=0, value=0.2}
         {arriveTangent=0.1443, interpMode=RCIM_Cubic, leaveTangent=0.1443, tangentMode=RCTM_User, time=1, value=1}
        ]}
       ZCurve {keys [
         {arriveTangent=5.9564, interpMode=RCIM_Cubic, leaveTangent=5.9564, tangentMode=RCTM_User, time=0, value=0.5}
         {arriveTangent=-0.1, interpMode=RCIM_Cubic, leaveTangent=-0.1, time=0.2, value=1}
         {arriveTangent=-1.1924, interpMode=RCIM_Cubic, leaveTangent=-1.1924, tangentMode=RCTM_User, time=1, value=0.4}
        ]}
      }})) {enabled=false}
   &ScaleMeshSize(Scale Factor=&MaskVectorBySpawnGroup(
     Masked Vector=&VectorFromCurve(
      VectorCurve=DataInterface() {propertyValues {
        XCurve {keys [
          {arriveTangent=2.2781, interpMode=RCIM_Cubic, leaveTangent=2.2781, tangentMode=RCTM_User, time=0, value=0.1}
          {arriveTangent=0.288, interpMode=RCIM_Cubic, leaveTangent=0.288, tangentMode=RCTM_User, time=1, value=1}
         ]}
        YCurve {keys [
          {arriveTangent=2.2781, interpMode=RCIM_Cubic, leaveTangent=2.2781, tangentMode=RCTM_User, time=0, value=0.1}
          {arriveTangent=0.288, interpMode=RCIM_Cubic, leaveTangent=0.288, tangentMode=RCTM_User, time=1, value=1}
         ]}
        ZCurve {keys [
          {arriveTangent=3.6292, interpMode=RCIM_Cubic, leaveTangent=3.6292, tangentMode=RCTM_User, time=0, value=0.5}
          {arriveTangent=-1.1338, interpMode=RCIM_Cubic, leaveTangent=-1.1338, tangentMode=RCTM_User, time=0.3, value=0.6}
          {arriveTangent=-0.1814, interpMode=RCIM_Cubic, leaveTangent=-0.1814, tangentMode=RCTM_User, time=1, value=0.4}
         ]}
       }}
      Scale Curve=<20 20 20>
     )
     Passthrough Vector=&VectorFromCurve(
      VectorCurve=DataInterface() {propertyValues {
        XCurve {keys [
          {arriveTangent=3.9002, interpMode=RCIM_Cubic, leaveTangent=3.9002, tangentMode=RCTM_User, time=0, value=0.2}
          {arriveTangent=1.2534, interpMode=RCIM_Cubic, leaveTangent=1.2534, tangentMode=RCTM_User, time=0.1743, value=0.6343}
          {arriveTangent=0.2006, interpMode=RCIM_Cubic, leaveTangent=0.2006, tangentMode=RCTM_User, time=1, value=1}
         ]}
        YCurve {keys [
          {arriveTangent=3.9002, interpMode=RCIM_Cubic, leaveTangent=3.9002, tangentMode=RCTM_User, time=0, value=0.2}
          {arriveTangent=1.2534, interpMode=RCIM_Cubic, leaveTangent=1.2534, tangentMode=RCTM_User, time=0.1743, value=0.6343}
          {arriveTangent=0.2006, interpMode=RCIM_Cubic, leaveTangent=0.2006, tangentMode=RCTM_User, time=1, value=1}
         ]}
        ZCurve {keys [
          {arriveTangent=8.292, interpMode=RCIM_Cubic, leaveTangent=8.292, tangentMode=RCTM_User, time=0, value=0.16}
          {arriveTangent=-1.3261, interpMode=RCIM_Cubic, leaveTangent=-1.3261, tangentMode=RCTM_User, time=0.07, value=0.6}
          <at 1 0.2>
         ]}
       }}
      Scale Curve=<20 20 20>
     )
    )) {module=ScaleMeshSize001}
   &DynamicMaterialParameters(
    Index 0 Param 1=&MaskFloatBySpawnGroup(
     Masked Float=1
     Passthrough Float=&FloatFromCurve(FloatCurve=DataInterface() {propertyValues {Curve {keys [{arriveTangent=-3.0733, interpMode=RCIM_Cubic, leaveTangent=-3.0733, tangentMode=RCTM_User, time=0, value=1}, <at 1 -0.75 RCIM_Cubic>]}}})
    )
    Index 0 Param 2=&MaskFloatBySpawnGroup(
     Masked Float=1.75
     Passthrough Float=&FloatFromCurve(FloatCurve=DataInterface() {propertyValues {Curve {keys [
          {arriveTangent=-2.0952, interpMode=RCIM_Cubic, leaveTangent=-2.0952, tangentMode=RCTM_User, time=0, value=1.5}
          {arriveTangent=-1.4744, interpMode=RCIM_Cubic, leaveTangent=-1.4744, tangentMode=RCTM_User, time=1, value=0.25}
         ]}}})
    )
    Index 0 Param 3=&FloatFromCurve(
     FloatCurve=DataInterface() {propertyValues {Curve {keys [
         {arriveTangent=-2.0927, interpMode=RCIM_Cubic, leaveTangent=-2.0927, tangentMode=RCTM_User, time=0, value=1}
         {interpMode=RCIM_Cubic, leaveTangent=-1.7693, time=1, value=0}
        ]}}}
     Scale Curve=0.25
    )
    Index 0 Param 4=&FloatFromCurve(FloatCurve=DataInterface() {propertyValues {Curve {keys [{arriveTangent=1.0328, interpMode=RCIM_Cubic, leaveTangent=1.0328, tangentMode=RCTM_User, time=0, value=0}, <at 1 1 RCIM_Cubic>]}}})
    Module.Param0WriteEnabled=true
    Module.Param1WriteEnabled=true
    Module.Param2WriteEnabled=true
    Module.Param3WriteEnabled=true
   )
   &ScaleColor(
    Scale Alpha=&FloatFromCurve(FloatCurve=DataInterface() {propertyValues {Curve {keys [
         {arriveTangent=26.1682, interpMode=RCIM_Cubic, leaveTangent=26.1682, tangentMode=RCTM_User, time=0, value=0}
         {interpMode=RCIM_Cubic, tangentMode=RCTM_User, time=0.1, value=1}
         {tangentMode=RCTM_User, time=1, value=0}
        ]}}})
    Scale RGB=&Lerp_Vector(
     A=<3 1 1.5>
     B=<0.5 1 1.8>
     Alpha=&FloatFromCurve(
      FloatCurve=DataInterface() {propertyValues {Curve {keys [{interpMode=RCIM_Cubic, tangentMode=RCTM_User, time=0.1, value=0}, {interpMode=RCIM_Cubic, tangentMode=RCTM_User, time=0.4, value=1}]}}}
      CurveIndex=$Emitter.Age
     )
    )
    ScaleA=true
    ScaleRGB=true
   )
  ]
  rendererClasses [&NiagaraMeshRendererProperties]
  renderers [{rendererClass=&NiagaraMeshRendererProperties, rendererIndex=0}]
 }
 GroundWaveSlow {
  emitterUpdateScript [
   &EmitterState(Life Cycle Mode=::NewEnumerator1, Loop Behavior=::NewEnumerator1, Loop Duration=1, UseLoopDelay=false)
   &SpawnBurst_Instantaneous(Use Spawn Probability=false, UseLoopCountLimit=false)
   /SetVariables(Emitter.Linear Color:LinearColor=&ColorFromCurve(
     ColorCurve=DataInterface() {propertyValues {
       AlphaCurve {keys [<at 0 1 RCIM_Cubic>,<at 1 0 RCIM_Cubic>]}
       BlueCurve {keys [<at 0 5.5 RCIM_Cubic>,<at 0.6229 40 RCIM_Cubic>]}
       GreenCurve {keys [<at 0 4.3 RCIM_Cubic>,<at 0.6229 13.2708 RCIM_Cubic>]}
       RedCurve {keys [<at 0 40 RCIM_Cubic>,<at 0.6229 3.4067 RCIM_Cubic>]}
      }}
     CurveIndex=$Emitter.NormalizedLoopAge
    )) {isSetParameters=true}
  ]
  particleSpawnScript [
   &InitializeParticle(
    Lifetime=1.5
    Mesh Scale=&Multiply_VectorByFloat(
     Vector=<1.8 1.8 0.25>
     Float=&FN_LOD_Distance_Remap(Start Distance=1000, End Distance=22500, Minimum Output=0.5, Maximum Output=1, Clamp Output=true)
    )
    Mesh Scale Mode=::NewEnumerator4
    UsePositionOffset=true
    Write Material Random=true
   )
   &InitialMeshOrientation(
    Mesh Orientation Mode=::NewEnumerator2
    Rotation=&RandomRangeVector(Minimum=<0 0 0>, Maximum=<0 0 1>)
    Use Rotation Vector=true
   )
  ]
  particleUpdateScript [
   &ParticleState()
   &ScaleMeshSize(Scale Factor=&Multiply_VectorByFloat(
     Vector=<1 1 1>
     Float=&FloatFromCurve(
      FloatCurve=DataInterface() {propertyValues {Curve {keys [{arriveTangent=5.6899, interpMode=RCIM_Cubic, leaveTangent=5.6899, tangentMode=RCTM_User, time=0, value=0.4}, <at 1 2 RCIM_Cubic>]}}}
      Scale Curve=60
     )
    ))
   &ScaleColor(
    Scale Alpha=&FloatFromCurve(
     FloatCurve=DataInterface() {propertyValues {Curve {keys [
         {arriveTangent=26.1682, interpMode=RCIM_Cubic, leaveTangent=26.1682, tangentMode=RCTM_User, time=0, value=1}
         {arriveTangent=-3.7386, interpMode=RCIM_Cubic, leaveTangent=-3.7386, tangentMode=RCTM_User, time=0.1, value=1}
         {tangentMode=RCTM_User, time=1, value=0}
        ]}}}
     Scale Curve=2
    )
    ScaleA=true
    ScaleRGB=true
   )
  ]
  rendererClasses [&NiagaraMeshRendererProperties]
  renderers [{rendererClass=&NiagaraMeshRendererProperties, rendererIndex=0}]
 }
 Ground_Ring001 {
  bEnabled=false
  emitterUpdateScript [
   &EmitterState()
   &SpawnBurst_Instantaneous(Use Spawn Probability=false, UseLoopCountLimit=false) {module=SpawnBurst_Instantaneous002}
  ]
  particleSpawnScript [
   /SetVariables(Particles.VisibilityTag=&SpawnGroup()) {isSetParameters=true, module=SetVariables_#2}
   &InitializeParticle(
    Lifetime=1.3
    Mesh Scale=&Multiply_Vector(
     A=&MaskVectorBySpawnGroup(Masked Vector=&Multiply_VectorByFloat(Vector=<1 14 14>, Float=5), Passthrough Vector=<100 100 100>)
     B=&Lerp_Vector(
      A=<1 1 1>
      B=&MaskVectorBySpawnGroup(Masked Vector=<1 4 4>, Passthrough Vector=<4 4 1>)
      Alpha=&FN_LOD_Distance_Remap(Start Distance=16000, End Distance=30000, Minimum Output=0, Maximum Output=1, Clamp Output=true)
     )
    )
    Mesh Scale Mode=::NewEnumerator4
    Position Offset=&MaskVectorBySpawnGroup(Masked Vector=<0 0 -5>, Passthrough Vector=<0 0 0>)
    UsePositionOffset=true
    Write Material Random=true
   )
   &InitialMeshOrientation(
    Mesh Orientation Mode=::NewEnumerator2
    Rotation=&MaskVectorBySpawnGroup(Masked Vector=<0 0.25 0>, Passthrough Vector=&RandomRangeVector(Minimum=<0 0 0>, Maximum=<0 0 1>))
    Use Rotation Vector=true
   )
  ]
  particleUpdateScript [
   &ParticleState()
   &ScaleMeshSize(Scale Factor=&MaskVectorBySpawnGroup(
     Masked Vector=&MakeVector(
      X=&FloatFromCurve(FloatCurve=DataInterface() {propertyValues {Curve {keys [
           {interpMode=RCIM_Cubic, tangentMode=RCTM_User, time=0, value=1}
           {arriveTangent=-4.2193, interpMode=RCIM_Cubic, leaveTangent=-4.2193, tangentMode=RCTM_User, time=0.5, value=0.25}
          ]}}})
      Y=&FloatFromCurve(
       FloatCurve=DataInterface() {propertyValues {Curve {keys [<at 0 0 RCIM_Cubic>, {arriveTangent=6.4238, tangentMode=RCTM_User, time=1, value=4}]}}}
       Scale Curve=2
      )
      Z=&FloatFromCurve(
       FloatCurve=DataInterface() {propertyValues {Curve {keys [<at 0 0 RCIM_Cubic>, {arriveTangent=6.4238, tangentMode=RCTM_User, time=1, value=4}]}}}
       Scale Curve=2
      )
     )
     Passthrough Vector=&MakeVector(
      X=&FloatFromCurve(FloatCurve=DataInterface() {propertyValues {Curve {keys [<at 0 0 RCIM_Cubic>, {arriveTangent=25.7175, interpMode=RCIM_Cubic, leaveTangent=25.7175, tangentMode=RCTM_User, time=1, value=12}]}}})
      Y=&FloatFromCurve(FloatCurve=DataInterface() {propertyValues {Curve {keys [<at 0 0 RCIM_Cubic>, {arriveTangent=25.7175, interpMode=RCIM_Cubic, leaveTangent=25.7175, tangentMode=RCTM_User, time=1, value=12}]}}})
      Z=&FloatFromCurve(FloatCurve=DataInterface() {propertyValues {Curve {keys [
           {arriveTangent=9.0725, interpMode=RCIM_Cubic, leaveTangent=9.0725, tangentMode=RCTM_User, time=0, value=0}
           {arriveTangent=0.0235, interpMode=RCIM_Cubic, leaveTangent=0.0235, tangentMode=RCTM_User, time=0.1408, value=2}
           {arriveTangent=-4.4427, interpMode=RCIM_Cubic, leaveTangent=-4.4427, tangentMode=RCTM_User, time=1, value=0}
          ]}}})
     )
    ))
   &ScaleColor(
    Scale Alpha=&MaskFloatBySpawnGroup(
     Masked Float=&FloatFromCurve(
      FloatCurve=DataInterface() {propertyValues {Curve {keys [
          {arriveTangent=-0.0176, interpMode=RCIM_Cubic, leaveTangent=-0.0176, tangentMode=RCTM_User, time=0, value=0.45}
          {arriveTangent=0.1545, interpMode=RCIM_Cubic, leaveTangent=0.1545, tangentMode=RCTM_User, time=0.3, value=1}
          {arriveTangent=0.0267, tangentMode=RCTM_User, time=1, value=0}
         ]}}}
      Scale Curve=0.07
     )
     Passthrough Float=&FloatFromCurve(
      FloatCurve=DataInterface() {propertyValues {Curve {keys [
          {arriveTangent=26.1682, interpMode=RCIM_Cubic, leaveTangent=26.1682, tangentMode=RCTM_User, time=0, value=1}
          {arriveTangent=-3.7386, interpMode=RCIM_Cubic, leaveTangent=-3.7386, tangentMode=RCTM_User, time=0.1, value=1}
          {tangentMode=RCTM_User, time=1, value=0}
         ]}}}
      Scale Curve=4
     )
    )
    Scale RGB=&MakeVectorFromLinearColorRGB(Color=&ColorFromCurve(ColorCurve=DataInterface() {propertyValues {
        AlphaCurve {keys []}
        BlueCurve {keys [<at 0.0387 5>,<at 0.9799 1>]}
        GreenCurve {keys [<at 0.0387 4.3292>,<at 0.9799 1>]}
        RedCurve {keys [<at 0.0387 0.1492>,<at 0.9799 1>]}
       }}))
    ScaleA=true
    ScaleRGB=true
   )
  ]
  rendererClasses [&NiagaraMeshRendererProperties]
  renderers [
   {rendererClass=&NiagaraMeshRendererProperties, rendererIndex=0}
   {rendererClass=&NiagaraMeshRendererProperties, rendererIndex=1}
  ]
 }
 Lightning_Ring {
  emitterUpdateScript [
   &EmitterState(Inactive Response=::NewEnumerator1, Life Cycle Mode=::NewEnumerator1, Loop Behavior=::NewEnumerator1, Loop Duration=8, UseLoopDelay=false)
   &SpawnRate(
    SpawnRate=&FloatFromCurve(FloatCurve=DataInterface() {propertyValues {Curve {keys [<at 0 1>,<at 0.2 0>]}}}, CurveIndex=$Emitter.Age, Scale Curve=15)
    Use Spawn Probability=false
   )
  ]
  particleSpawnScript [
   &PartitionParticles()
   &InitializeParticle(
    Color=<rgba 0.0938 0.4511 1 1>
    Mesh Scale=<0.5 0.5 0.5>
    Mesh Scale Mode=::NewEnumerator4
    Position Offset=&MakeVector(X=0, Y=0, Z=&RandomRangeFloat(Minimum=0, Maximum=-500))
    UsePositionOffset=true
    Write Material Random=true
   )
   &InitialMeshOrientation(
    Mesh Orientation Mode=::NewEnumerator2
    Rotation=&RandomRangeVector(Minimum=<0 0 0>, Maximum=<0 0 1>)
    Use Rotation Vector=true
   )
   &SpriteFacingAndAlignment(Sprite Facing=<0 1 0>, WriteAlignment=false, WriteFacing=true)
  ]
  particleUpdateScript [
   &ParticleState()
   &ScaleMeshSize(Scale Factor=&VectorFromFloat(Value=&Multiply_Float(
      A=&SelectFloatFromArray(Direct Array Index=$Particles.PartitionParticles.Partition, Float Selection Array=DataInterface() {propertyValues {FloatData [1.8,1.3]}})
      B=&FloatFromCurve(
       FloatCurve=DataInterface() {propertyValues {Curve {keys [{arriveTangent=1.3195, interpMode=RCIM_Cubic, leaveTangent=1.3195, tangentMode=RCTM_User, time=0, value=0}, <at 0.9 1 RCIM_Cubic>]}}}
       Scale Curve=&Multiply_Float(
        A=200
        B=&FN_LOD_Distance_Remap(Start Distance=350, End Distance=380, Minimum Output=0.7, Maximum Output=1, Clamp Output=true)
       )
      )
     )))
   &ScaleColor(
    Scale Alpha=&Multiply_Float(
     A=&FNTimeOfDayScalar(Night Scale=0.7, Morning Scale=1, Evening Scale=1)
     B=&FloatFromCurve(FloatCurve=DataInterface() {propertyValues {Curve {keys [<at 0.8 1>,<at 0.9 0>]}}})
    )
    ScaleA=true
    ScaleRGB=true
   )
   /SetVariables(Particles.Color002=&MultiplyLinearColor(
     Linear Color=<rgba 0.1926 0.0697 0.6115 0.2>
     Scale Factor=/Niagara/DynamicInputs/Multiply/Multiply_Vector4_ByFloat(
      Vector4:Vector4f=<1 1 1 1>
      Float:NiagaraFloat=&Multiply_Float(
       A=&FNTimeOfDayScalar(Night Scale=0.5, Morning Scale=1, Evening Scale=1)
       B=&FloatFromCurve(
        FloatCurve=DataInterface() {propertyValues {Curve {keys [<at 0.8 1>,<at 0.9 0>]}}}
        Scale Curve=&FN_FrontEndFloatModifier(Gameplay=1, Front End=0.5)
       )
      )
     )
    )) {isSetParameters=true}
   &DynamicMaterialParameters(
    Index 3 Param 3=&FN_LOD_Distance_Remap(
     Start Distance=400
     End Distance=1000
     Minimum Output=&FloatFromCurve(FloatCurve=DataInterface() {propertyValues {Curve {defaultValue=1, keys [{tangentMode=RCTM_User, time=0, value=0}, {tangentMode=RCTM_User, time=0.5, value=10}]}}})
     Maximum Output=&FloatFromCurve(FloatCurve=DataInterface() {propertyValues {Curve {defaultValue=1, keys [{tangentMode=RCTM_User, time=0, value=0}, {tangentMode=RCTM_User, time=0.5, value=20}]}}})
     Clamp Output=true
    )
    Module.Index3Param0Write=false
    Module.Index3Param1Write=false
    Module.Index3Param2Write=true
    Module.Index3Param3Write=false
    Write Parameter Index 0=false
    Write Parameter Index 3=true
   )
  ]
  rendererClasses [&NiagaraMeshRendererProperties]
  renderers [
   {rendererClass=&NiagaraMeshRendererProperties, rendererIndex=0}
   {rendererClass=&NiagaraMeshRendererProperties, rendererIndex=1}
  ]
 }
 Wave02 {
  bEnabled=false
  emitterUpdateScript [
   &EmitterState(Inactive Response=::NewEnumerator1, Life Cycle Mode=::NewEnumerator1, Loop Behavior=::NewEnumerator1, Loop Duration=3.17, UseLoopDelay=false)
   &SpawnRate(Use Spawn Probability=false) {enabled=false}
   &SpawnBurst_Instantaneous(Spawn Count=1, Use Spawn Probability=false)
  ]
  particleSpawnScript [
   &InitializeParticle(
    Color=<rgba 3 0 2.1179 1>
    Lifetime=2
    Mesh Scale Mode=::NewEnumerator3
    Mesh Uniform Scale=11
    Sprite Rotation Mode=::NewEnumerator2
    UsePositionOffset=true
    Write Material Random=true
   )
   &InitialMeshOrientation(Orientation Axis=<0 0 1>, Orientation Vector=<0 0 1>, Use Orientation Vector=true, Use Rotation Vector=true)
   &InitialMeshOrientation(
    Orientation Axis=<0 0 0>
    Orientation Vector=<0 0 0>
    Rotation=&RandomRangeVector(Minimum=<0 0 0>, Maximum=<0 1 0>)
    Use Orientation Vector=false
    Use Rotation Vector=true
   ) {enabled=false, module=InitialMeshOrientation001}
   &CameraOffset()
  ]
  particleUpdateScript [
   &ParticleState()
   &ScaleColor(
    Linear Color Curve=DataInterface() {propertyValues {
      AlphaCurve {keys [<at 0 1>,<at 0.3576 1>,<at 1 0>]}
      BlueCurve {keys [<at 0 0.0414>,<at 0.1684 0.45>,<at 0.3654 1>,<at 1 1>]}
      GreenCurve {keys [<at 0 1>,<at 0.1684 1>,<at 0.3654 0.701>,<at 1 0.0574>]}
      RedCurve {keys [<at 0 0.2012>,<at 0.1684 0>,<at 0.3654 0.1029>,<at 1 0.0414>]}
     }}
    Scale Mode=::NewEnumerator2
   ) {enabled=false}
   &DynamicMaterialParameters(
    Index 0 Param 1=&FloatFromCurve(FloatCurve=DataInterface() {propertyValues {Curve {keys [{arriveTangent=-2.1015, interpMode=RCIM_Cubic, leaveTangent=-2.1015, tangentMode=RCTM_User, time=0, value=0.5}, <at 1 -0.15>]}}})
    Index 0 Param 2=&FloatFromCurve(FloatCurve=DataInterface() {propertyValues {Curve {keys [{arriveTangent=-0.3464, interpMode=RCIM_Cubic, leaveTangent=-0.3464, tangentMode=RCTM_User, time=0, value=1}, <at 1 0.9>]}}})
    Index 0 Param 3=&FloatFromCurve(FloatCurve=DataInterface() {propertyValues {Curve {keys [<at 0 0>,<at 0.15 1>,<at 1 1.2>]}}})
    Module.Param0WriteEnabled=true
    Module.Param1WriteEnabled=true
    Module.Param2WriteEnabled=true
    Module.Param3WriteEnabled=false
   )
   &Color(Color=&LerpLinearColors(
     StartColor=<rgba 0.0226 0.6786 1 1>
     EndColor=<rgba 0.76 0.037 1 1>
     LerpFactor=&FloatFromCurve(FloatCurve=DataInterface() {propertyValues {Curve {
         defaultValue=1
         keys [{arriveTangent=0.0881, interpMode=RCIM_Cubic, leaveTangent=0.0881, tangentMode=RCTM_User, time=0, value=0}, <at 1 1 RCIM_Cubic>]
        }}})
    ))
   &ScaleColor(
    Scale Alpha=&FloatFromCurve(
     FloatCurve=DataInterface() {propertyValues {Curve {keys [<at 0.1 0 RCIM_Cubic>, {interpMode=RCIM_Cubic, leaveTangent=-2.4889, tangentMode=RCTM_Break, time=0.3047, value=0.8}, <at 1 0>]}}}
     Scale Curve=&FN_FrontEndFloatModifier(Gameplay=1, Front End=0.6)
    )
    Scale RGB=&VectorFromFloat(Value=&FloatFromCurve(FloatCurve=DataInterface() {propertyValues {Curve {keys [<at 0 10>,<at 0.15 3>,<at 1 1>]}}}))
    ScaleA=true
    ScaleRGB=false
   ) {module=ScaleColor001}
   &ScaleMeshSize(Scale Factor=&VectorFromFloat(Value=&FloatFromCurve(
      FloatCurve=DataInterface() {propertyValues {Curve {keys [<at 0 0>,<at 0.1 1.4>,<at 1 2>]}}}
      Scale Curve=&Multiply_Float(
       A=0.115
       B=&Multiply_Float(
        A=$System.SizeScale
        B=&FN_LOD_Distance_Remap(Start Distance=350, End Distance=380, Minimum Output=0.7, Maximum Output=1, Clamp Output=true)
       )
      )
     )))
  ]
  rendererClasses [&NiagaraMeshRendererProperties]
  renderers [{rendererClass=&NiagaraMeshRendererProperties, rendererIndex=0}]
 }
}
summary {
 systemName=NS_PistolCanary_TripleTentacleWave
 emitters [
  {emitterName=Energy_Sphere, rendererClasses [&NiagaraMeshRendererProperties]}
  {bEnabled=false, emitterName=Ground_Ring001, rendererClasses [&NiagaraMeshRendererProperties]}
  {
   emitterName=Cone_Shock_Light
   rendererClasses [
    &NiagaraMeshRendererProperties
    &NiagaraSpriteRendererProperties
    &NiagaraLightRendererProperties
   ]
  }
  {emitterName=Bolt, rendererClasses [&NiagaraMeshRendererProperties]}
  {bEnabled=false, emitterName=Wave02, rendererClasses [&NiagaraMeshRendererProperties]}
  {
   bEnabled=false
   emitterName=Dark_Glow
   rendererClasses [
    &NiagaraLightRendererProperties
    &NiagaraSpriteRendererProperties
    &NiagaraMeshRendererProperties
   ]
  }
  {emitterName=Lightning_Ring, rendererClasses [&NiagaraMeshRendererProperties]}
  {bEnabled=false, emitterName=GroundWave, rendererClasses [&NiagaraMeshRendererProperties]}
  {emitterName=GroundWaveSlow, rendererClasses [&NiagaraMeshRendererProperties]}
  {emitterName=Explode_Sparkles, rendererClasses [&NiagaraSpriteRendererProperties]}
  {emitterName=Flash, rendererClasses [&NiagaraMeshRendererProperties]}
  {emitterName=EnergyWave, rendererClasses [&NiagaraSpriteRendererProperties, &NiagaraLightRendererProperties]}
 ]
 userVariables [
  User.bIsFrontEnd {type=NiagaraBool}
  User.bIsFrontEndPreview {type=NiagaraBool}
 ]
}
